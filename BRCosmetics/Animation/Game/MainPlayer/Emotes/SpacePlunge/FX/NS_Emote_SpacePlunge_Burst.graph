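# Niagara burst effect for the SpacePlunge emote, built around a skeletal mesh of cubes and timed from a shared System.Delay.
# Cubes pins 16 mesh particles to bones with a curve-driven outline.
# Dark_Glow layers sprites, meshes and lights through spawn-group switches, and Flare adds delayed flashes.
# User.Variant picks orange, blue or red palettes, front-end and time-of-day modifiers adjust brightness, and Dark_Glow001 and Energy are disabled.
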NiagaraSystem /BRCosmetics/Animation/Game/MainPlayer/Emotes/SpacePlunge/FX/NS_Emote_SpacePlunge_Burst
dependencies {
 usedDataInterfaces [
  &NiagaraDataInterfaceArrayColor
  &NiagaraDataInterfaceCurve
  &NiagaraDataInterfaceSpriteRendererInfo
  &NiagaraDataInterfaceColorCurve
  &NiagaraDataInterfaceCamera
  &NiagaraDataInterfaceVector2DCurve
  &NiagaraDataInterfaceVectorCurve
  &NiagaraDataInterfaceArrayFloat
  &NiagaraDataInterfaceSkeletalMesh
 ]
 usedDynamicInputs [&SelectFloatFromArray]
 usedModules [
  &self:SystemSpawnScript.SystemScriptSource.SystemScriptGraph.NiagaraNodeAssignment_2.SetVariables_#1
  &self:SystemSpawnScript.SystemScriptSource.SystemScriptGraph.NiagaraNodeAssignment_1.SetVariables_#2
  &self:SystemSpawnScript.SystemScriptSource.SystemScriptGraph.NiagaraNodeAssignment_3.SetVariables_#3
  &self:SystemSpawnScript.SystemScriptSource.SystemScriptGraph.NiagaraNodeAssignment_4.SetVariables_#4
  &self:SystemSpawnScript.SystemScriptSource.SystemScriptGraph.NiagaraNodeAssignment_5.SetVariables_#5
  &self:SystemSpawnScript.SystemScriptSource.SystemScriptGraph.NiagaraNodeAssignment_0.SetVariables_#6
  &SystemState
 ]
}
emitters {
 Cubes {
  emitterUpdateScript [
   &EmitterState(Life Cycle Mode=::NewEnumerator1, Loop Behavior=::NewEnumerator1, Loop Delay=$System.Delay, Loop Duration=10, UseLoopDelay=true)
   &SpawnBurst_Instantaneous(Spawn Count=16, Use Spawn Probability=false, UseLoopCountLimit=false)
  ]
  particleSpawnScript [
   /SetVariables(Particles.OutlineVis:NiagaraInt32=&MakeIntFromFloat(Float=$User.Variant, Conversion Mode=::NewEnumerator1)) {isSetParameters=true}
   &PartitionParticles(Number Of Partitions=16, Particles Per Partition=1, Partition Mode:ENiagara_PartitionMode_v2=::NewEnumerator1)
   &InitializeParticle(
    Color Mode=::NewEnumerator0
    Lifetime=10
    Mesh Scale Mode=::NewEnumerator3
    Mesh Uniform Scale=1.05
    UsePositionOffset=false
    Write Material Random=true
   )
  ]
  particleUpdateScript [
   &ParticleState()
   /SetVariables(Particles.OutlineColor:LinearColor=&MakeLinearColor(
     R:NiagaraFloat=4
     G:NiagaraFloat=4
     B:NiagaraFloat=4
     A:NiagaraFloat=&FloatFromCurve(
      FloatCurve=DataInterface() {propertyValues {Curve {
         defaultValue=0
         keys [
          {arriveTangent=0.2583, interpMode=RCIM_Cubic, leaveTangent=0.2583, tangentMode=RCTM_User, time=0.1136, value=0}
          {arriveTangent=2.0899, interpMode=RCIM_Cubic, leaveTangent=2.0899, time=0.4718, value=0.2554}
          {arriveTangentWeight=0.1342, interpMode=RCIM_Cubic, leaveTangentWeight=0.1657, tangentMode=RCTM_User, time=0.5921, value=1}
          {arriveTangentWeight=0.1599, interpMode=RCIM_Cubic, leaveTangentWeight=0.1771, tangentMode=RCTM_User, time=0.8841, value=0.6701}
          {arriveTangent=-6.0797, interpMode=RCIM_Cubic, leaveTangent=-6.0797, tangentMode=RCTM_User, time=0.9405, value=0}
         ]
        }}}
      CurveIndex=$Emitter.NormalizedLoopAge
     )
    )) {isSetParameters=true}
   &SkeletalMeshLocation(
    Bone / Socket Index=$Particles.PartitionParticles.Partition
    Bone Sampling Mode=::NewEnumerator4
    MaskSpawnGroup=false
    Sampled Position Coordinate Space=::Local
    Skeletal Mesh=$System.Cubes
   )
   &ScaleMeshSize(Scale Factor=$Particles.SkeletalMeshLocation.SampledScale)
  ]
  rendererClasses [&NiagaraMeshRendererProperties]
  renderers [
   {rendererClass=&NiagaraMeshRendererProperties, rendererIndex=0}
   {rendererClass=&NiagaraMeshRendererProperties, rendererIndex=1}
   {rendererClass=&NiagaraMeshRendererProperties, rendererIndex=2}
  ]
 }
 Dark_Glow {
  emitterUpdateScript [
   &EmitterState(
    Inactive Response=::NewEnumerator1
    Life Cycle Mode=::NewEnumerator1
    Loop Behavior=::NewEnumerator1
    Loop Delay=&Add_Float(A=$System.Delay, B=0.7)
    Loop Duration=9
    UseLoopDelay=true
   )
   &SpawnRate(
    SpawnGroup=6
    SpawnRate=&FloatFromCurve(
     FloatCurve=DataInterface() {propertyValues {Curve {
        defaultValue=1
        keys [
         {interpMode=RCIM_Cubic, tangentMode=RCTM_User, time=2.4457, value=0}
         {arriveTangent=0.4476, interpMode=RCIM_Cubic, leaveTangent=0.4476, time=3.9377, value=0.2565}
         {arriveTangent=1.6993, interpMode=RCIM_Cubic, leaveTangent=-0.0872, tangentMode=RCTM_Break, time=4.68, value=1}
         <at 6.1875 1 RCIM_Constant>
         <at 6.6973 0 RCIM_Constant>
        ]
       }}}
     CurveIndex=$Emitter.Age
     Scale Curve=30
    )
    Use Spawn Probability=false
   )
   &SpawnBurst_Instantaneous(Spawn Group=5, Spawn Time=8.5, Use Spawn Probability=false, UseLoopCountLimit=false) {module=SpawnBurst_Instantaneous005}
   &SpawnBurst_Instantaneous(Spawn Group=4, Spawn Time=0.8, Use Spawn Probability=false, UseLoopCountLimit=false) {module=SpawnBurst_Instantaneous004}
   &SpawnBurst_Instantaneous(Spawn Group=3, Spawn Time=1.3, Use Spawn Probability=false, UseLoopCountLimit=false) {module=SpawnBurst_Instantaneous003}
   &SpawnBurst_Instantaneous(Spawn Group=2, Use Spawn Probability=false, UseLoopCountLimit=false) {module=SpawnBurst_Instantaneous002}
   &SpawnBurst_Instantaneous(Spawn Group=1, Spawn Time=4.8, Use Spawn Probability=false, UseLoopCountLimit=false) {module=SpawnBurst_Instantaneous001}
   &SpawnBurst_Instantaneous(Spawn Count=1, Spawn Time=0.8, Use Spawn Probability=false)
  ]
  particleSpawnScript [
   /SetVariables(Particles.VisibilityTag=&SpawnGroup()) {isSetParameters=true, module=SetVariables_#7}
   /SetVariables(Particles.Random=&RandomRangeFloat(Minimum=0, Maximum=1)) {isSetParameters=true}
   &InitializeParticle(
    Color=&MergeHelper_ColorSwitch(
     0=<rgba 1 1 1 1.5>
     1=<rgba 1 1 1 1.5>
     2=&SelectLinearColorFromArray(
      Array Sampling Mode=::NewEnumerator1
      Direct Array Index=&MakeIntFromFloat(Float=$User.Variant, Conversion Mode=::NewEnumerator1)
      Color Selection Array=DataInterface() {propertyValues {ColorData [<rgba 1 0.3784 0 1>, <rgba 0 0.4916 1 1>, <rgba 1 0.0238 0.024 1>]}}
      Random Seed=?
     )
     3=&SelectLinearColorFromArray(
      Array Sampling Mode=::NewEnumerator1
      Direct Array Index=&MakeIntFromFloat(Float=$User.Variant, Conversion Mode=::NewEnumerator1)
      Color Selection Array=DataInterface() {propertyValues {ColorData [<rgba 1 0.3784 0 1>, <rgba 0 0.4916 1 1>, <rgba 1 0.0167 0 1>]}}
      Random Seed=?
     )
     4=&MergeHelper_ColorSwitch(
      0=<rgba 3 0.965 0 1>
      1=<rgba 0 1.0468 2 1>
      2=<rgba 3 0.0468 0 1>
      3=?
      Index=&MakeIntFromFloat(Float=$User.Variant, Conversion Mode=::NewEnumerator1)
      Num=::NewEnumerator1
     )
     5=&SelectLinearColorFromArray(
      Array Sampling Mode=::NewEnumerator1
      Direct Array Index=&MakeIntFromFloat(Float=$User.Variant, Conversion Mode=::NewEnumerator1)
      Color Selection Array=DataInterface() {propertyValues {ColorData [<rgba 1 0.298 0 1>, <rgba 0 0.3144 1 1>, <rgba 1 0.0001 0.0167 1>]}}
      Random Seed=?
     )
     6=&MergeHelper_ColorSwitch(
      0=<rgba 5 3.6617 0 1>
      1=<rgba 0 3.1555 5 1>
      2=<rgba 5 0.1522 0.1148 1>
      3=?
      Index=&MakeIntFromFloat(Float=$User.Variant, Conversion Mode=::NewEnumerator1)
      Num=::NewEnumerator1
     )
     Index=&SpawnGroup()
     Num=::NewEnumerator5
    )
    Lifetime=&MergeHelper_FloatSwitch(0=4.7, 1=4.7, 2=3, 3=8, 4=5, 5=0.5, 6=0.25, Index=&SpawnGroup(), Num=::NewEnumerator5)
    Mesh Scale=&MergeHelper_VectorSwitch(
     0=<0 0 0>
     1=<0 0 0>
     2=<0 0 0>
     3=<0 0 0>
     4=&Multiply_VectorByFloat(Vector=<-2 -4 -3.5>, Float=0.3)
     5=&Multiply_VectorByFloat(Vector=<-2 -4 -3.5>, Float=0.3)
     6=&Multiply_VectorByFloat(
      Vector=&Lerp_Vector(
       A=&RandomRangeVector(Minimum=<0.5 0.4 0.4>, Maximum=<0.75 0.6 0.6>)
       B=&RandomRangeVector(Minimum=<1 1 1>, Maximum=<1.5 2 2>)
       Alpha=&FloatFromCurve(
        FloatCurve=DataInterface() {propertyValues {Curve {keys [{interpMode=RCIM_Cubic, tangentMode=RCTM_User, time=0.5, value=0}, <at 1 1>]}}}
        CurveIndex=&RandomRangeFloat(Minimum=0, Maximum=1)
       )
      )
      Float=0.6
     )
     Index=&SpawnGroup()
     Num=::NewEnumerator5
    )
    Mesh Scale Mode=::NewEnumerator4
    Sprite Rotation Angle=&MergeHelper_FloatSwitch(0=0.12, 1=0.12, 2=0, 3=0, 4=0, Index=&SpawnGroup(), Num=::NewEnumerator3)
    Sprite Rotation Mode=::NewEnumerator1
    Sprite Size Mode=::NewEnumerator3
    Uniform Sprite Size=&MergeHelper_FloatSwitch(0=100, 1=100, 2=60, 3=120, 4=0, Index=&SpawnGroup(), Num=::NewEnumerator3)
    UsePositionOffset=false
    Write Material Random=true
   )
   &DynamicMaterialParameters(
    Index 0 Param 1=&MergeHelper_FloatSwitch(0=2, 1=2, 2=3, 3=1, 4=0, 5=0, Index=&SpawnGroup(), Num=::NewEnumerator4)
    Index 0 Param 2=&MergeHelper_FloatSwitch(0=1, 1=1, 2=1, 3=1, 4=1, 5=0, Index=&SpawnGroup(), Num=::NewEnumerator4)
    Index 0 Param 3=&MergeHelper_FloatSwitch(0=1, 1=1, 2=1, 3=1, 4=1, 5=0, Index=&SpawnGroup(), Num=::NewEnumerator4)
    Index 0 Param 4=&MergeHelper_FloatSwitch(0=1, 1=1, 2=1, 3=1, 4=1, 5=0, Index=&SpawnGroup(), Num=::NewEnumerator4)
    Module.Index0Float0Write=true
    Module.Index0Float1Write=true
    Module.Index0Float2Write=true
    Module.Index0Float3Write=true
   )
   &CameraOffset(Camera Offset Amount=&MergeHelper_FloatSwitch(0=10, 1=10, 2=0, 3=0, 4=0, Index=&SpawnGroup(), Num=::NewEnumerator3))
   &SpriteFacingAndAlignment(Sprite Facing=<0 1 0>, WriteAlignment=false, WriteFacing=true)
   &AddVelocity(
    Module.Use Falloff Distance=false
    Override Seed=false
    Velocity Mode=::NewEnumerator1
    Velocity Speed=&MergeHelper_FloatSwitch(0=0, 1=0, 2=0, 3=0, 4=0, 5=0, 6=-20, Index=&SpawnGroup(), Num=::NewEnumerator5)
   )
  ]
  particleUpdateScript [
   &ParticleState()
   &ScaleMeshSize(Scale Factor=&MergeHelper_VectorSwitch(
     0=<0 0 0>
     1=<0 0 0>
     2=<0 0 0>
     3=<0 0 0>
     4=&VectorFromFloat(Value=&FloatFromCurve(FloatCurve=DataInterface() {propertyValues {Curve {
          defaultValue=1
          keys [{interpMode=RCIM_Cubic, tangentMode=RCTM_User, time=0.9682, value=1}, {interpMode=RCIM_Cubic, tangentMode=RCTM_User, time=1, value=0}]
         }}}))
     5=&VectorFromFloat(Value=&FloatFromCurve(FloatCurve=DataInterface() {propertyValues {Curve {
          defaultValue=1
          keys [{interpMode=RCIM_Cubic, tangentMode=RCTM_User, time=0.9682, value=1}, {interpMode=RCIM_Cubic, tangentMode=RCTM_User, time=1, value=0}]
         }}}))
     6=&VectorFromCurve(VectorCurve=DataInterface() {propertyValues {
        XCurve {keys [<at 0.2848 1 RCIM_Cubic>, {arriveTangent=-3.0417, interpMode=RCIM_Cubic, leaveTangent=-3.0417, tangentMode=RCTM_User, time=1, value=0}]}
        YCurve {keys [<at 0 1>,<at 1 1>]}
        ZCurve {keys [<at 0 1>,<at 1 1>]}
       }})
     Index=&SpawnGroup()
     Num=::NewEnumerator5
    ))
   /SetVariables(Particles.DynPar_ChromaWave:Vector4f=&MakeVector4(
     X=&RandomRangeFloat(Minimum=0, Maximum=1)
     Y=&MergeHelper_FloatSwitch(
      0=0
      1=0
      2=0
      3=0
      4=&FloatFromCurve(
       FloatCurve=DataInterface() {propertyValues {Curve {
          defaultValue=1
          keys [
           {arriveTangent=0.2806, interpMode=RCIM_Cubic, leaveTangent=0.2806, tangentMode=RCTM_User, time=0, value=0}
           {arriveTangent=2.2839, interpMode=RCIM_Cubic, leaveTangent=2.2839, tangentMode=RCTM_User, time=1, value=1}
          ]
         }}}
       Scale Curve=2
      )
      5=&FloatFromCurve(
       FloatCurve=DataInterface() {propertyValues {Curve {
          defaultValue=1
          keys [
           {arriveTangent=-3.4505, interpMode=RCIM_Cubic, leaveTangent=-3.4505, leaveTangentWeight=0.8311, tangentMode=RCTM_User, time=0, value=1}
           {arriveTangent=-0.2806, arriveTangentWeight=0.3462, interpMode=RCIM_Cubic, leaveTangent=-0.2806, tangentMode=RCTM_User, time=1, value=0}
          ]
         }}}
       Scale Curve=0.7
      )
      Index=&SpawnGroup()
      Num=::NewEnumerator4
     )
     Z=0
     W=0
    )) {isSetParameters=true}
   /SetVariables(Particles.Lightning glow color:LinearColor=&MakeLinearColorFromVectorAndFloat(
     "Vector (RGB)"=&MakeVectorFromLinearColorRGB(Color=&MultiplyLinearColor(
       Linear Color=&SelectLinearColorFromArray(
        Array Sampling Mode=::NewEnumerator1
        Direct Array Index=&MakeIntFromFloat(Float=$User.Variant, Conversion Mode=::NewEnumerator1)
        Color Selection Array=DataInterface() {propertyValues {ColorData [<rgba 4 1.6071 0.1968 1>, <rgba 0.3366 2.0701 4 1>, <rgba 4 0.1073 0.1894 1>]}}
        Random Seed=?
       )
       Scale Factor=&MakeVector4_FromFloat(Float=&RandomRangeFloat(Minimum=0.5, Maximum=1, Evaluation Type=::NewEnumerator1))
      ))
     "Float (Alpha)"=&Multiply_Float(
      A=&FloatFromCurve(FloatCurve=DataInterface() {propertyValues {Curve {keys [<at 0 1 RCIM_Cubic>, {interpMode=RCIM_Cubic, leaveTangent=-3.6307, time=0.9, value=0}]}}})
      B=1
     )
    )) {isSetParameters=true}
   /SetVariables(Particles.Glow Dy02:Vector4f=&MakeVector4(
     X=&FloatFromCurve(FloatCurve=DataInterface() {propertyValues {Curve {keys [{tangentMode=RCTM_User, time=0, value=0.2}, {tangentMode=RCTM_User, time=0.5, value=1}]}}})
     Y=0.7
     Z=0
     W=5
    )) {isSetParameters=true}
   /SetVariables(Particles.Glow Dy03:Vector4f=&MakeVector4(
     X=&FloatFromCurve(
      FloatCurve=DataInterface() {propertyValues {Curve {keys [<at 0 1>,<at 1 0.5>]}}}
      Scale Curve=&MergeHelper_FloatSwitch(0=0.7, 1=0.5, 2=1.8, 3=?, Index=&MakeIntFromFloat(Float=$User.Variant), Num=::NewEnumerator1)
     )
     Y=$Particles.Random
     Z=1.3
     W=1
    )) {isSetParameters=true}
   /SetVariables(
    Particles.Lightning Dyn Param:Vector4f=&MakeVector4(
     X=&FloatFromCurve(FloatCurve=DataInterface() {propertyValues {Curve {keys [{tangentMode=RCTM_User, time=0, value=1}, {tangentMode=RCTM_User, time=1, value=1}]}}})
     Y=0.7
     Z=&FloatFromCurve(FloatCurve=DataInterface() {propertyValues {Curve {keys [
          {arriveTangent=4.0072, interpMode=RCIM_Cubic, leaveTangent=4.0072, tangentMode=RCTM_User, time=0, value=0}
          {arriveTangent=2.7083, interpMode=RCIM_Cubic, leaveTangent=2.7083, time=0.2543, value=0.6225}
          {arriveTangent=1.8473, interpMode=RCIM_Cubic, leaveTangent=1.8473, time=0.6, value=1.625}
          {arriveTangent=1.6993, interpMode=RCIM_Cubic, leaveTangent=1.6993, tangentMode=RCTM_User, time=1, value=2}
         ]}}})
     W=5
    )
    Particles.Lightning Dyn Param 01:Vector4f=&MakeVector4(
     X=&MergeHelper_FloatSwitch(
      0=0.7
      1=0.7
      2=0.5
      3=?
      Index=&MakeIntFromFloat(Float=$User.Variant, Conversion Mode=::NewEnumerator1)
      Num=::NewEnumerator1
     )
     Y=$Particles.Random
     Z=1.3
     W=1
    )
    Particles.Lightning Dyn Param 02:Vector4f=&MakeVector4(
     X=&FloatFromCurve(FloatCurve=DataInterface() {propertyValues {Curve {keys [<at 0 1>,<at 1 0.3>]}}})
     Y=0
     Z=0
     W=0
    )
   ) {isSetParameters=true}
   &SolveForcesAndVelocity(Clamp Velocity=false, Limit Acceleration=false)
   &SpriteRotationRate(Rotation Rate=&MergeHelper_FloatSwitch(
     0=0
     1=0
     2=0
     3=&FloatFromCurve(
      FloatCurve=DataInterface() {propertyValues {Curve {
         defaultValue=1
         keys [
          {interpMode=RCIM_Cubic, tangentMode=RCTM_User, time=0, value=0}
          {arriveTangent=-4.4276, interpMode=RCIM_Cubic, leaveTangent=-4.4276, time=0.0958, value=1.1104}
          {arriveTangentWeight=0.1138, interpMode=RCIM_Cubic, leaveTangentWeight=0.2881, tangentMode=RCTM_User, time=0.2911, value=-1.2889}
          {arriveTangentWeight=0.2866, interpMode=RCIM_Cubic, leaveTangentWeight=0.5185, tangentMode=RCTM_User, time=0.6308, value=0}
          <at 0.7688 0.18 RCIM_Cubic>
          {interpMode=RCIM_Cubic, tangentMode=RCTM_User, time=1, value=0}
         ]
        }}}
      Scale Curve=400
     )
     4=0
     Index=&SpawnGroup()
     Num=::NewEnumerator3
    ))
   &ScaleColor(
    Scale Alpha=&MergeHelper_FloatSwitch(
     0=&Multiply_Float(
      A=&FN_LOD_Distance_Remap(Start Distance=350, End Distance=380, Minimum Output=1, Maximum Output=1, Clamp Output=true)
      B=&FloatFromCurve(
       FloatCurve=DataInterface() {propertyValues {Curve {keys [
           <at 0.1602 0 RCIM_Cubic>
           {arriveTangentWeight=0.2318, interpMode=RCIM_Cubic, leaveTangentWeight=0.2396, tangentMode=RCTM_User, time=0.5985, value=1}
           <at 1 1 RCIM_Cubic>
          ]}}}
       Scale Curve=1.1
      )
     )
     1=&Multiply_Float(
      A=&FN_LOD_Distance_Remap(Start Distance=350, End Distance=380, Minimum Output=1, Maximum Output=1, Clamp Output=true)
      B=&FloatFromCurve(
       FloatCurve=DataInterface() {propertyValues {Curve {keys [
           <at 0.1602 0 RCIM_Cubic>
           {arriveTangentWeight=0.2318, interpMode=RCIM_Cubic, leaveTangentWeight=0.2396, tangentMode=RCTM_User, time=0.5985, value=1}
           <at 1 1 RCIM_Cubic>
          ]}}}
       Scale Curve=1.1
      )
     )
     2=&FloatFromCurve(
      FloatCurve=DataInterface() {propertyValues {Curve {
         defaultValue=1
         keys [
          {arriveTangent=6.8265, interpMode=RCIM_Cubic, leaveTangent=6.8265, tangentMode=RCTM_User, time=0, value=0}
          {interpMode=RCIM_Cubic, tangentMode=RCTM_User, time=0.5, value=1}
          <at 1 0 RCIM_Cubic>
         ]
        }}}
      Scale Curve=&FN_FrontEndFloatModifier(Gameplay=0.5, Front End=0.3)
     )
     3=&FloatFromCurve(
      FloatCurve=DataInterface() {propertyValues {Curve {keys [
          <at 0 0.5 RCIM_Cubic>
          {arriveTangentWeight=0.0865, interpMode=RCIM_Cubic, leaveTangentWeight=0.2862, tangentMode=RCTM_User, time=0.2321, value=1}
          {arriveTangentWeight=0.2434, interpMode=RCIM_Cubic, leaveTangentWeight=0.1554, tangentMode=RCTM_User, time=0.985, value=1}
         ]}}}
      Scale Curve=&FN_FrontEndFloatModifier(Gameplay=1, Front End=0.5)
     )
     4=&FloatFromCurve(FloatCurve=DataInterface() {propertyValues {Curve {
         defaultValue=1
         keys [
          {arriveTangent=0.1462, interpMode=RCIM_Cubic, leaveTangent=0.1462, tangentMode=RCTM_User, time=0, value=0}
          {interpMode=RCIM_Cubic, tangentMode=RCTM_User, time=0.7269, value=1}
         ]
        }}})
     5=&FloatFromCurve(FloatCurve=DataInterface() {propertyValues {Curve {
         defaultValue=0
         keys [{interpMode=RCIM_Cubic, tangentMode=RCTM_User, time=0, value=1}, {interpMode=RCIM_Cubic, tangentMode=RCTM_User, time=1, value=0}]
        }}})
     6=&Multiply_Float(
      A=10
      B=&FloatFromCurve(FloatCurve=DataInterface() {propertyValues {Curve {keys [
           {interpMode=RCIM_Cubic, tangentMode=RCTM_User, time=0, value=0}
           {interpMode=RCIM_Cubic, tangentMode=RCTM_User, time=0.1, value=0}
           {interpMode=RCIM_Cubic, tangentMode=RCTM_User, time=0.2, value=0.8}
           {interpMode=RCIM_Cubic, tangentMode=RCTM_User, time=1, value=0}
          ]}}})
     )
     Index=&SpawnGroup()
     Num=::NewEnumerator5
    )
    Scale RGB=&MergeHelper_VectorSwitch(
     0=<1 1 1>
     1=<1 1 1>
     2=<1 1 1>
     3=&VectorFromFloat(Value=&FloatFromCurve(FloatCurve=DataInterface() {propertyValues {Curve {keys [
           {interpMode=RCIM_Cubic, tangentMode=RCTM_User, time=0, value=0.5}
           <at 0.5 1 RCIM_Cubic>
           {arriveTangent=-0.1284, interpMode=RCIM_Cubic, leaveTangent=-0.1284, tangentMode=RCTM_User, time=1, value=0.5}
          ]}}}))
     4=<1 1 1>
     5=<1 1 1>
     6=&Multiply_VectorByFloat(
      Vector=&Multiply_VectorByFloat(Vector=<1 1 1>, Float=&FN_FrontEndFloatModifier(Gameplay=1, Front End=0.5))
      Float=&Multiply_Float(
       A=&RandomRangeFloat(Minimum=0.5, Maximum=1, Evaluation Type=::NewEnumerator1)
       B=&Multiply_Float(A=&FNTimeOfDayScalar(Night Scale=0.2, Morning Scale=1, Evening Scale=1), B=10)
      )
     )
     Index=&SpawnGroup()
     Num=::NewEnumerator5
    )
    ScaleA=true
    ScaleRGB=true
   )
   &ScaleSpriteSize(
    Scale Factor=&MergeHelper_Vector2Switch(
     0=<1 1>
     1=<1 1>
     2=<1 1>
     3=&Vector2DFromFloat(Value=&FloatFromCurve(FloatCurve=DataInterface() {propertyValues {Curve {keys [
           {interpMode=RCIM_Cubic, tangentMode=RCTM_User, time=0, value=0.5}
           {
            arriveTangent=0.9101
            arriveTangentWeight=0.1572
            interpMode=RCIM_Cubic
            leaveTangent=0.9101
            leaveTangentWeight=0.187
            tangentMode=RCTM_User
            time=0.2013
            value=0.7311
           }
           {arriveTangentWeight=0.0486, interpMode=RCIM_Cubic, leaveTangentWeight=0.0063, tangentMode=RCTM_User, time=0.2316, value=0.6814}
           {arriveTangentWeight=0.0022, interpMode=RCIM_Cubic, leaveTangentWeight=0.0079, tangentMode=RCTM_User, time=0.2582, value=0.8592}
           {arriveTangentWeight=0.0079, interpMode=RCIM_Cubic, leaveTangentWeight=0.005, tangentMode=RCTM_User, time=0.2794, value=0.7085}
           {arriveTangentWeight=0.0056, interpMode=RCIM_Cubic, leaveTangentWeight=0.0103, tangentMode=RCTM_User, time=0.2947, value=1.0562}
           {arriveTangentWeight=0.0098, interpMode=RCIM_Cubic, leaveTangentWeight=0.0053, tangentMode=RCTM_User, time=0.3173, value=0.8333}
           {arriveTangentWeight=0.0153, interpMode=RCIM_Cubic, leaveTangentWeight=0.0391, tangentMode=RCTM_User, time=0.3257, value=1.2148}
           {arriveTangentWeight=0.2165, interpMode=RCIM_Cubic, leaveTangentWeight=0.1618, tangentMode=RCTM_User, time=0.3479, value=1.328}
           {arriveTangentWeight=0.1546, interpMode=RCIM_Cubic, leaveTangentWeight=0.0794, tangentMode=RCTM_User, time=0.3599, value=0.9473}
           {arriveTangentWeight=0.0037, interpMode=RCIM_Cubic, leaveTangentWeight=0.003, tangentMode=RCTM_User, time=0.3772, value=1.2876}
           {arriveTangentWeight=0.003, interpMode=RCIM_Cubic, leaveTangentWeight=0.0044, tangentMode=RCTM_User, time=0.3892, value=0.9119}
           {arriveTangentWeight=0.0044, interpMode=RCIM_Cubic, leaveTangentWeight=0.0039, tangentMode=RCTM_User, time=0.3981, value=1.849}
           {arriveTangentWeight=0.0039, interpMode=RCIM_Cubic, leaveTangentWeight=0.0028, tangentMode=RCTM_User, time=0.4083, value=1.285}
           {arriveTangentWeight=0.0028, interpMode=RCIM_Cubic, leaveTangentWeight=0.0048, tangentMode=RCTM_User, time=0.4181, value=1.7376}
           {arriveTangentWeight=0.0048, interpMode=RCIM_Cubic, leaveTangentWeight=0.0024, tangentMode=RCTM_User, time=0.4338, value=1.4407}
           {arriveTangentWeight=0.146, interpMode=RCIM_Cubic, leaveTangentWeight=0.1771, tangentMode=RCTM_User, time=0.4486, value=1.9071}
           {arriveTangentWeight=0.0029, interpMode=RCIM_Cubic, leaveTangentWeight=0.0043, tangentMode=RCTM_User, time=0.4668, value=1.3099}
           {arriveTangentWeight=0.014, interpMode=RCIM_Cubic, leaveTangentWeight=0.0241, tangentMode=RCTM_User, time=0.4777, value=1.9292}
           {arriveTangentWeight=0.0505, interpMode=RCIM_Cubic, leaveTangentWeight=0.0235, tangentMode=RCTM_User, time=0.5033, value=2.2433}
           {arriveTangentWeight=0.0186, interpMode=RCIM_Cubic, leaveTangentWeight=0.0293, tangentMode=RCTM_User, time=0.5163, value=0.896}
           {arriveTangentWeight=0.0007, interpMode=RCIM_Cubic, leaveTangentWeight=0.1139, tangentMode=RCTM_User, time=0.525, value=2.2846}
           {arriveTangentWeight=0.0104, interpMode=RCIM_Cubic, leaveTangentWeight=0.1242, tangentMode=RCTM_User, time=0.5633, value=1.7}
           {arriveTangent=0.5015, interpMode=RCIM_Cubic, leaveTangent=0.5015, time=0.6173, value=1.6127}
           {arriveTangent=-0.0268, interpMode=RCIM_Cubic, leaveTangent=-0.0268, time=0.7032, value=1.7701}
           {arriveTangent=-0.3546, interpMode=RCIM_Cubic, leaveTangent=-0.3546, time=0.8235, value=1.6072}
           {interpMode=RCIM_Cubic, tangentMode=RCTM_User, time=0.901, value=1.7}
           {arriveTangentWeight=0.0737, interpMode=RCIM_Cubic, leaveTangentWeight=0.0932, tangentMode=RCTM_User, time=0.92, value=2.2798}
           {arriveTangentWeight=0.6437, interpMode=RCIM_Cubic, tangentMode=RCTM_User, time=0.935, value=0}
          ]}}}))
     4=<0 0>
     Index=&SpawnGroup()
     Num=::NewEnumerator3
    )
    Scale Sprite Size Mode=::NewEnumerator2
   )
   &FN_Sprite_Attributes(
    Color=&MultiplyLinearColor(
     Linear Color=&MergeHelper_ColorSwitch(
      0=<rgba 1.2 0.1281 0 1>
      1=<rgba 0 0.2487 1.2 1>
      2=<rgba 1.2 0 0.0069 1>
      3=?
      Index=&MakeIntFromFloat(Float=$User.Variant, Conversion Mode=::NewEnumerator1)
      Num=::NewEnumerator1
     )
     Scale Factor=&MakeVector4(
      X=1
      Y=1
      Z=1
      W=&MaskFloatBySpawnGroup(
       Masked Float=&FloatFromCurve(FloatCurve=DataInterface() {propertyValues {Curve {
           defaultValue=1
           keys [{interpMode=RCIM_Cubic, tangentMode=RCTM_User, time=0.1749, value=0}, {interpMode=RCIM_Cubic, tangentMode=RCTM_User, time=0.6968, value=1}]
          }}})
       Passthrough Float=&FloatFromCurve(FloatCurve=DataInterface() {propertyValues {Curve {
           defaultValue=1
           keys [
            {interpMode=RCIM_Cubic, tangentMode=RCTM_User, time=0.1749, value=0}
            {interpMode=RCIM_Cubic, tangentMode=RCTM_User, time=0.1896, value=1}
            {arriveTangentWeight=0.3461, interpMode=RCIM_Cubic, leaveTangentWeight=0.0814, tangentMode=RCTM_User, time=0.4257, value=1}
            <at 0.8293 0 RCIM_Cubic>
           ]
          }}})
      )
     )
    )
    Create Color Attribute=true
    Create SpriteSize Attribute=true
    SpriteSize=&Multiply_Vector2D(
     A=&Vector2DFromFloat(Value=150)
     B=&Vector2DFromFloat(Value=&Multiply_Float(
       A=&MaskFloat(
        Mask Source=0
        Mask Parameter=?
        Mask=0
        Masked Float=&FloatFromCurve(FloatCurve=DataInterface() {propertyValues {Curve {keys [<at 0 0>,<at 0.05 1>,<at 0.83 1>,<at 0.9 0>]}}})
        Passthrough Float=&FloatFromCurve(FloatCurve=DataInterface() {propertyValues {Curve {keys [
             {arriveTangent=6.2432, interpMode=RCIM_Cubic, leaveTangent=6.2432, tangentMode=RCTM_User, time=0, value=0}
             {arriveTangentWeight=0.1385, interpMode=RCIM_Cubic, leaveTangentWeight=0.2947, tangentMode=RCTM_User, time=0.2592, value=1}
            ]}}})
       )
       B=&Multiply_Float(
        A=1
        B=&FN_LOD_Distance_Remap(Start Distance=350, End Distance=380, Minimum Output=0.7, Maximum Output=1, Clamp Output=true)
       )
      ))
    )
    Write to Module Namespace=true
   )
   &SkeletalMeshLocation(
    Bone / Socket Index=&MergeHelper_IntSwitch(0=0, 1=0, 2=11, 3=0, 4=0, 5=0, 6=0, Index=&SpawnGroup(), Num=::NewEnumerator5)
    Bone Sampling Mode=::NewEnumerator4
    MaskSpawnGroup=false
    Skeletal Mesh=$System.Cubes
   )
   /SetVariables(Particles.MeshOrientation=&OrientMeshToVector(
     Look At Direction=&ReturnCameraVector(Camera Vectors=::NewEnumerator1, Coordinate Space=::Simulation)
     Reference Axis Vector=<1 0 0>
    )) {isSetParameters=true}
   &Light_Attributes(
    Color=1
    Color Value To Scale:LinearColor=&Multiply_LinearColorByFloat(
     Linear Color=&SelectLinearColorFromArray(
      Array Sampling Mode=::NewEnumerator1
      Direct Array Index=&MakeIntFromFloat(Float=$User.Variant, Conversion Mode=::NewEnumerator1)
      Color Selection Array=DataInterface() {propertyValues {ColorData [<rgba 20 9.2925 0 1>, <rgba 0 8.3625 20 1>, <rgba 20 0.4792 0.3724 1>]}}
      Random Seed=?
     )
     Float=1
    )
    Light Radius=170
    Radius=1
    Scale Alpha:NiagaraFloat=&FloatFromCurve(FloatCurve=DataInterface() {propertyValues {Curve {
        defaultValue=1
        keys [
         {arriveTangent=0.1755, interpMode=RCIM_Cubic, leaveTangent=0.1755, tangentMode=RCTM_User, time=0, value=0}
         {
          arriveTangent=0.8794
          arriveTangentWeight=0.1038
          interpMode=RCIM_Cubic
          leaveTangent=0.8794
          leaveTangentWeight=0.3962
          tangentMode=RCTM_User
          time=0.5081
          value=0.2389
         }
         {arriveTangent=16.759, interpMode=RCIM_Cubic, leaveTangent=16.759, time=0.5244, value=2.1742}
         {arriveTangent=-3.0713, interpMode=RCIM_Cubic, leaveTangent=-3.0713, time=0.5327, value=0.6514}
         {arriveTangentWeight=0.1897, interpMode=RCIM_Cubic, leaveTangentWeight=0.1744, tangentMode=RCTM_User, time=0.539, value=2.1295}
         {arriveTangentWeight=0.0372, interpMode=RCIM_Cubic, leaveTangentWeight=0.2083, tangentMode=RCTM_User, time=0.5464, value=1}
         {arriveTangent=-4.2228, interpMode=RCIM_Cubic, leaveTangent=-4.2228, tangentMode=RCTM_User, time=0.9477, value=0.5527}
         {interpMode=RCIM_Cubic, leaveTangent=-1.1146, time=0.9499, value=0}
        ]
       }}})
    ScaleA=true
    ScaleRGB=false
   )
   &ShapeLocation(
    MaskSpawnGroup=false
    Override Seed=false
    Shape Origin=$Particles.SkeletalMeshLocation.SampledPosition
    Sphere Radius=10
    Sphere Surface Distribution=1
    UseOwnerScale=true
   )
  ]
  rendererClasses [
   &NiagaraSpriteRendererProperties
   &NiagaraLightRendererProperties
   &NiagaraMeshRendererProperties
  ]
  renderers [
   {rendererClass=&NiagaraSpriteRendererProperties}
   {rendererClass=&NiagaraSpriteRendererProperties, rendererIndex=1}
   {rendererClass=&NiagaraSpriteRendererProperties, rendererIndex=2}
   {rendererClass=&NiagaraSpriteRendererProperties, rendererIndex=3}
   {rendererClass=&NiagaraLightRendererProperties, rendererIndex=4}
   {rendererClass=&NiagaraMeshRendererProperties, rendererIndex=5}
   {rendererClass=&NiagaraMeshRendererProperties, rendererIndex=6}
   {rendererClass=&NiagaraMeshRendererProperties, rendererIndex=7}
   {rendererClass=&NiagaraMeshRendererProperties, rendererIndex=8}
  ]
 }
 Dark_Glow001 {
  bEnabled=false
  emitterUpdateScript [
   &EmitterState(
    Inactive Response=::NewEnumerator1
    Life Cycle Mode=::NewEnumerator1
    Loop Behavior=::NewEnumerator1
    Loop Delay=&Add_Float(A=$System.Delay, B=1.5)
    Loop Duration=9
    UseLoopDelay=true
   )
   &SpawnBurst_Instantaneous(Spawn Group=1, Spawn Time=4, Use Spawn Probability=false, UseLoopCountLimit=false) {module=SpawnBurst_Instantaneous001}
   &SpawnBurst_Instantaneous(Spawn Count=1, Use Spawn Probability=false)
  ]
  particleSpawnScript [
   /SetVariables(Particles.VisibilityTag=&SpawnGroup()) {isSetParameters=true, module=SetVariables_#7}
   &InitializeParticle(
    Color=<rgba 1 1 1 1.5>
    Lifetime=4.7
    Sprite Rotation Angle=0.12
    Sprite Rotation Mode=::NewEnumerator1
    Sprite Size Mode=::NewEnumerator3
    Uniform Sprite Size=100
    UsePositionOffset=false
    Write Material Random=true
   )
   &DynamicMaterialParameters(
    Index 0 Param 1=2
    Index 0 Param 2=1
    Index 0 Param 3=1
    Index 0 Param 4=1
    Module.Index0Float0Write=true
    Module.Index0Float1Write=true
    Module.Index0Float2Write=true
    Module.Index0Float3Write=true
   )
   &CameraOffset(Camera Offset Amount=10)
   &SpriteFacingAndAlignment(Sprite Facing=<0 1 0>, WriteAlignment=false, WriteFacing=true)
  ]
  particleUpdateScript [
   &ParticleState()
   &SolveForcesAndVelocity(Clamp Velocity=false, Limit Acceleration=false)
   &ScaleColor(
    Scale Alpha=&Multiply_Float(
     A=&FN_LOD_Distance_Remap(Start Distance=350, End Distance=380, Minimum Output=1, Maximum Output=1, Clamp Output=true)
     B=&FloatFromCurve(
      FloatCurve=DataInterface() {propertyValues {Curve {keys [
          <at 0.1602 0 RCIM_Cubic>
          {arriveTangentWeight=0.2318, interpMode=RCIM_Cubic, leaveTangentWeight=0.2396, tangentMode=RCTM_User, time=0.5985, value=1}
          <at 1 1 RCIM_Cubic>
         ]}}}
      Scale Curve=1.1
     )
    )
    Scale RGB=&MakeVectorFromLinearColorRGB(Color=&ColorFromCurve(ColorCurve=DataInterface() {propertyValues {
        AlphaCurve {keys []}
        BlueCurve {keys [<at 0 0.0917 RCIM_Cubic>]}
        GreenCurve {keys [<at 0 0.1 RCIM_Cubic>]}
        RedCurve {keys [<at 0 0.0219 RCIM_Cubic>]}
       }}))
    ScaleA=true
    ScaleRGB=false
   )
   &FN_Sprite_Attributes(
    Color=&MultiplyLinearColor(
     Linear Color=&MergeHelper_ColorSwitch(0=<rgba 1.2 0.1281 0 1>, 1=<rgba 0 0.2487 1.2 1>, 2=<rgba 1.2 0.1281 0 1>, 3=?, Index=0, Num=::NewEnumerator1)
     Scale Factor=&MakeVector4(
      X=1
      Y=1
      Z=1
      W=&MaskFloatBySpawnGroup(
       Masked Float=&FloatFromCurve(FloatCurve=DataInterface() {propertyValues {Curve {
           defaultValue=1
           keys [{interpMode=RCIM_Cubic, tangentMode=RCTM_User, time=0.1749, value=0}, {interpMode=RCIM_Cubic, tangentMode=RCTM_User, time=0.6968, value=1}]
          }}})
       Passthrough Float=&FloatFromCurve(FloatCurve=DataInterface() {propertyValues {Curve {
           defaultValue=1
           keys [
            {interpMode=RCIM_Cubic, tangentMode=RCTM_User, time=0.1749, value=0}
            {interpMode=RCIM_Cubic, tangentMode=RCTM_User, time=0.1896, value=1}
            {arriveTangentWeight=0.3461, interpMode=RCIM_Cubic, leaveTangentWeight=0.0814, tangentMode=RCTM_User, time=0.4257, value=1}
            <at 0.8293 0 RCIM_Cubic>
           ]
          }}})
      )
     )
    )
    Create Color Attribute=true
    Create SpriteSize Attribute=true
    SpriteSize=&Multiply_Vector2D(
     A=&Vector2DFromFloat(Value=150)
     B=&Vector2DFromFloat(Value=&Multiply_Float(
       A=&MaskFloat(
        Mask Source=0
        Mask Parameter=?
        Mask=0
        Masked Float=&FloatFromCurve(FloatCurve=DataInterface() {propertyValues {Curve {keys [<at 0 0>,<at 0.05 1>,<at 0.83 1>,<at 0.9 0>]}}})
        Passthrough Float=&FloatFromCurve(FloatCurve=DataInterface() {propertyValues {Curve {keys [
             {arriveTangent=6.2432, interpMode=RCIM_Cubic, leaveTangent=6.2432, tangentMode=RCTM_User, time=0, value=0}
             {arriveTangentWeight=0.1385, interpMode=RCIM_Cubic, leaveTangentWeight=0.2947, tangentMode=RCTM_User, time=0.2592, value=1}
            ]}}})
       )
       B=&Multiply_Float(
        A=1
        B=&FN_LOD_Distance_Remap(Start Distance=350, End Distance=380, Minimum Output=0.7, Maximum Output=1, Clamp Output=true)
       )
      ))
    )
    Write to Module Namespace=true
   )
   &SkeletalMeshLocation(Bone Sampling Mode=::NewEnumerator4, MaskSpawnGroup=false, Skeletal Mesh=$System.Cubes)
  ]
  rendererClasses [&NiagaraSpriteRendererProperties]
  renderers [
   {rendererClass=&NiagaraSpriteRendererProperties}
   {rendererClass=&NiagaraSpriteRendererProperties, rendererIndex=1}
  ]
 }
 Energy {
  bEnabled=false
  emitterUpdateScript [
   &EmitterState(
    Life Cycle Mode=::NewEnumerator1
    Loop Behavior=::NewEnumerator1
    Loop Delay=&Add_Float(A=$System.Delay, B=9.2)
    UseLoopDelay=true
   )
   &SpawnBurst_Instantaneous(Use Spawn Probability=false, UseLoopCountLimit=false)
  ]
  particleSpawnScript [&InitializeParticle(
    Color=&SelectLinearColorFromArray(
     Array Sampling Mode=::NewEnumerator1
     Direct Array Index=0
     Color Selection Array=DataInterface() {propertyValues {ColorData [<rgba 1 0.298 0 1>, <rgba 0 0.3144 1 1>, <rgba 1 0.298 0 1>]}}
     Random Seed=?
    )
    Lifetime=0.5
    Mesh Scale=&Multiply_VectorByFloat(Vector=<-2 -4 -3.5>, Float=0.3)
    Mesh Scale Mode=::NewEnumerator4
    UsePositionOffset=false
    Write Material Random=true
   )]
  particleUpdateScript [
   &ParticleState()
   &ScaleMeshSize(Scale Factor=&VectorFromFloat(Value=&FloatFromCurve(FloatCurve=DataInterface() {propertyValues {Curve {
         defaultValue=1
         keys [{interpMode=RCIM_Cubic, tangentMode=RCTM_User, time=0.9682, value=1}, {interpMode=RCIM_Cubic, tangentMode=RCTM_User, time=1, value=0}]
        }}})))
   &UpdateMeshOrientation(
    Facing Interpolation=false
    Facing Position:NiagaraPosition=&ReturnCameraPosition(Coordinate Space:ENiagaraCoordinateSpace=::Simulation, Camera Query:NiagaraDataInterfaceCamera=DataInterface())
    Orientation Method=::NewEnumerator2
    Override Quat=false
    TransformFacingOrigin=false
    TransformFacingPosition=false
    TransformUp=false
   ) {enabled=false}
   /SetVariables(
    Particles.DynPar_ChromaWave:Vector4f=&MakeVector4(
     X=&RandomRangeFloat(Minimum=0, Maximum=1)
     Y=&FloatFromCurve(
      FloatCurve=DataInterface() {propertyValues {Curve {
         defaultValue=1
         keys [
          {arriveTangent=-3.4505, interpMode=RCIM_Cubic, leaveTangent=-3.4505, leaveTangentWeight=0.8311, tangentMode=RCTM_User, time=0, value=1}
          {arriveTangent=-0.2806, arriveTangentWeight=0.3462, interpMode=RCIM_Cubic, leaveTangent=-0.2806, tangentMode=RCTM_User, time=1, value=0}
         ]
        }}}
      Scale Curve=0.7
     )
     Z=0
     W=0
    )
    Particles.DynPar_SoftWave:Vector4f=&MakeVector4(X=3, Y=1.5, Z=0, W=0)
   ) {isSetParameters=true, module=SetVariables_#8}
   &SkeletalMeshLocation(Bone Sampling Mode=::NewEnumerator4, MaskSpawnGroup=false, Skeletal Mesh=$System.Cubes)
   /SetVariables(Particles.MeshOrientation=&OrientMeshToVector(
     Look At Direction=&ReturnCameraVector(Camera Vectors=::NewEnumerator1, Coordinate Space=::Simulation)
     Reference Axis Vector=<1 0 0>
    )) {isSetParameters=true, module=SetVariables_#9}
   &ScaleColor(
    Scale Alpha=&FloatFromCurve(FloatCurve=DataInterface() {propertyValues {Curve {
        defaultValue=0
        keys [{interpMode=RCIM_Cubic, tangentMode=RCTM_User, time=0, value=1}, {interpMode=RCIM_Cubic, tangentMode=RCTM_User, time=1, value=0}]
       }}})
    ScaleA=true
    ScaleRGB=false
   )
   &DynamicMaterialParameters(Module.Index0Float0Write=true, Module.Index0Float1Write=true, Module.Index0Float2Write=true, Module.Index0Float3Write=true) {enabled=false}
  ]
  rendererClasses [&NiagaraMeshRendererProperties]
  renderers [{rendererClass=&NiagaraMeshRendererProperties, rendererIndex=0}]
 }
 Flare {
  emitterUpdateScript [
   &EmitterState(
    Life Cycle Mode=::NewEnumerator1
    Loop Behavior=::NewEnumerator1
    Loop Delay=&Add_Float(A=$System.Delay, B=6.1)
    UseLoopDelay=true
   )
   &SpawnBurst_Instantaneous(Spawn Group=1, Spawn Time=3.1, Use Spawn Probability=false, UseLoopCountLimit=false) {module=SpawnBurst_Instantaneous001}
   &SpawnBurst_Instantaneous(Use Spawn Probability=false, UseLoopCountLimit=false)
  ]
  particleSpawnScript [
   &InitializeParticle(
    Color=&MakeCustomLinearColorFromBool(Bool=$User.bIsFrontEndPreview, True Color=<rgba 0.3185 0.0137 0 1.3942>, False Color=<rgba 0.3177 0.0139 0 3>)
    Lifetime=0.3
    Position Offset=<0 0 -10>
    Sprite Size Mode=::NewEnumerator3
    Uniform Sprite Size=150
    UsePositionOffset=false
    Write Material Random=true
   )
   /SetVariables(Particles.Vector 4:Vector4f=&MakeVector4(X=&RandomRangeFloat(Minimum=0.7, Maximum=1), Y=1, Z=1, W=1)) {isSetParameters=true}
   &CameraOffset(Camera Offset Amount=20)
  ]
  particleUpdateScript [
   &ParticleState()
   &SpriteRotationRate(Rotation Rate=&FloatFromCurve(
     FloatCurve=DataInterface() {propertyValues {Curve {keys [
         {interpMode=RCIM_Cubic, tangentMode=RCTM_User, time=0, value=1}
         {arriveTangent=-1.2549, interpMode=RCIM_Cubic, leaveTangent=-1.2549, tangentMode=RCTM_User, time=1, value=0}
        ]}}}
     Scale Curve=20
    ))
   &ScaleColor(
    Linear Color Curve=DataInterface() {propertyValues {
      AlphaCurve {keys [<at 0 0>,<at 0.0244 1>,<at 0.0395 0.5829>,<at 0.9564 0.1804>,<at 0.97 0.6>,<at 1 0>]}
      BlueCurve {keys [<at 0 1>]}
      GreenCurve {keys [<at 0 1>]}
      RedCurve {keys [<at 0 1>]}
     }}
    Scale Mode=::NewEnumerator2
   )
   /SetVariables(
    Particles.FlareColor=&MergeHelper_ColorSwitch(
     0=&ColorFromCurve(ColorCurve=DataInterface() {propertyValues {
        AlphaCurve {keys [
          <at 0 0>
          {arriveTangent=11.236, interpMode=RCIM_Cubic, leaveTangent=11.236, time=0.022, value=1}
          <at 0.089 1>
          <at 0.157 0.25>
          {arriveTangentWeight=0.0649, interpMode=RCIM_Cubic, leaveTangentWeight=0.1733, tangentMode=RCTM_User, time=0.22, value=0}
          <at 0.51 0.0074>
          <at 0.55 0.4435>
          <at 0.6 0.5655>
          <at 0.7 0>
         ]}
        BlueCurve {keys [<at 0 0 RCIM_Cubic>,<at 0.35 0 RCIM_Cubic>,<at 0.6464 0>,<at 0.8448 0>]}
        GreenCurve {keys [
          <at 0 4.2659 RCIM_Cubic>
          {arriveTangentWeight=0.1997, interpMode=RCIM_Cubic, leaveTangentWeight=1.9287, tangentMode=RCTM_User, time=0.35, value=0.051}
          <at 0.4 0.051>
          <at 0.6 5.3387>
         ]}
        RedCurve {keys [
          <at 0 10 RCIM_Cubic>
          {arriveTangentWeight=0.3189, interpMode=RCIM_Cubic, leaveTangentWeight=3.0805, tangentMode=RCTM_User, time=0.35, value=1}
          <at 0.4 1>
          <at 0.6 10>
         ]}
       }})
     1=&ColorFromCurve(ColorCurve=DataInterface() {propertyValues {
        AlphaCurve {keys [
          <at 0 0>
          {arriveTangent=11.236, interpMode=RCIM_Cubic, leaveTangent=11.236, time=0.022, value=1}
          <at 0.089 1>
          <at 0.157 0.25>
          {arriveTangentWeight=0.0649, interpMode=RCIM_Cubic, leaveTangentWeight=0.1733, tangentMode=RCTM_User, time=0.22, value=0}
          <at 0.51 0.0074>
          <at 0.55 0.4435>
          <at 0.6 0.5655>
          <at 0.7 0>
         ]}
        BlueCurve {keys [<at 0 10 RCIM_Cubic>,<at 0.3474 1>,<at 0.4525 10>]}
        GreenCurve {keys [<at 0 7.1331 RCIM_Cubic>,<at 0.3474 0.2196>,<at 0.4525 7.1331>]}
        RedCurve {keys [<at 0 0 RCIM_Cubic>,<at 0.3474 0>,<at 0.4525 0>]}
       }})
     2=&ColorFromCurve(ColorCurve=DataInterface() {propertyValues {
        AlphaCurve {keys [
          <at 0 0>
          {arriveTangent=11.236, interpMode=RCIM_Cubic, leaveTangent=11.236, time=0.022, value=1}
          <at 0.089 1>
          <at 0.157 0.25>
          {arriveTangentWeight=0.0649, interpMode=RCIM_Cubic, leaveTangentWeight=0.1733, tangentMode=RCTM_User, time=0.22, value=0}
          <at 0.51 0.0074>
          <at 0.55 0.4435>
          <at 0.6 0.5655>
          <at 0.7 0>
         ]}
        BlueCurve {keys [<at 0 0 RCIM_Cubic>,<at 0.35 0 RCIM_Cubic>,<at 0.6464 0>,<at 0.8448 0>]}
        GreenCurve {keys [
          <at 0 0.0179 RCIM_Cubic>
          {arriveTangentWeight=0.1997, interpMode=RCIM_Cubic, leaveTangentWeight=1.9287, tangentMode=RCTM_User, time=0.35, value=0.015}
          <at 0.4 0.051>
          <at 0.6 0>
         ]}
        RedCurve {keys [
          <at 0 10 RCIM_Cubic>
          {arriveTangentWeight=0.3189, interpMode=RCIM_Cubic, leaveTangentWeight=3.0805, tangentMode=RCTM_User, time=0.35, value=1}
          <at 0.4 1>
          <at 0.6 10>
         ]}
       }})
     3=?
     Index=&MakeIntFromFloat(Float=$User.Variant, Conversion Mode=::NewEnumerator1)
     Num=::NewEnumerator1
    )
    Particles.FlareDynPar:Vector4f=<1 0 0 0>
    Particles.FlareSpriteSize:Vector2f=&Vector2DFromFloat(Value=&FloatFromCurve(
      FloatCurve=DataInterface() {propertyValues {Curve {keys [
          {interpMode=RCIM_Cubic, leaveTangent=5.6981, tangentMode=RCTM_Break, time=0, value=0.2}
          <at 0.315 1>
          {arriveTangent=-0.209, interpMode=RCIM_Cubic, leaveTangent=-0.209, time=4.9, value=1}
          {arriveTangent=-13.1443, interpMode=RCIM_Cubic, leaveTangent=-13.1443, tangentMode=RCTM_User, time=5.1, value=0}
         ]}}}
      Scale Curve=&MaskFloatBySpawnGroup(Masked Float=200, Passthrough Float=150)
     ))
   ) {isSetParameters=true}
   /SetVariables(
    Particles.RaysColor:LinearColor=&MergeHelper_ColorSwitch(
     0=&ColorFromCurve(ColorCurve=DataInterface() {propertyValues {
        AlphaCurve {keys [
          <at 0 0>
          {arriveTangent=32.3708, interpMode=RCIM_Cubic, leaveTangent=32.3708, time=0.0127, value=0.7091}
          <at 0.0309 1>
          <at 0.8487 0.6091>
          {arriveTangent=-0.5244, interpMode=RCIM_Cubic, leaveTangent=-0.5244, tangentMode=RCTM_User, time=0.97, value=0}
         ]}
        BlueCurve {keys [
          <at 0 0 RCIM_Cubic>
          {arriveTangent=0.1333, interpMode=RCIM_Cubic, leaveTangent=0.1333, time=0.025, value=0.1293}
          <at 0.8503 0>
          <at 0.97 0.1293>
         ]}
        GreenCurve {keys [
          <at 0 2.3556 RCIM_Cubic>
          {arriveTangent=-1.4871, interpMode=RCIM_Cubic, leaveTangent=-1.4871, time=0.025, value=0.9131}
          <at 0.8503 3.294>
          <at 0.97 0.9131>
         ]}
        RedCurve {keys [
          <at 0 6 RCIM_Cubic>
          {arriveTangent=-4.6392, interpMode=RCIM_Cubic, leaveTangent=-4.6392, time=0.025, value=1.5}
          <at 0.8503 6>
          <at 0.97 1.5>
         ]}
       }})
     1=&ColorFromCurve(ColorCurve=DataInterface() {propertyValues {
        AlphaCurve {keys [
          <at 0 0>
          {arriveTangent=32.3708, interpMode=RCIM_Cubic, leaveTangent=32.3708, time=0.0127, value=0.7091}
          <at 0.0309 1>
          <at 0.8487 0.6091>
          {arriveTangent=-0.5244, interpMode=RCIM_Cubic, leaveTangent=-0.5244, tangentMode=RCTM_User, time=0.97, value=0}
         ]}
        BlueCurve {keys [<at 0 6 RCIM_Cubic>,<at 0.025 1.5 RCIM_Cubic>,<at 0.8503 6>,<at 0.97 1.5>]}
        GreenCurve {keys [
          <at 0 4.8009 RCIM_Cubic>
          {arriveTangent=-3.311, interpMode=RCIM_Cubic, leaveTangent=-3.311, time=0.025, value=0.4544}
          <at 0.8503 1.9856>
          <at 0.97 1.0201>
         ]}
        RedCurve {keys [<at 0 0 RCIM_Cubic>,<at 0.025 0 RCIM_Cubic>,<at 0.8503 0>,<at 0.97 0>]}
       }})
     2=&ColorFromCurve(ColorCurve=DataInterface() {propertyValues {
        AlphaCurve {keys [
          <at 0 0>
          {arriveTangent=32.3708, interpMode=RCIM_Cubic, leaveTangent=32.3708, time=0.0127, value=0.7091}
          <at 0.0309 1>
          <at 0.8487 0.6091>
          {arriveTangent=-0.5244, interpMode=RCIM_Cubic, leaveTangent=-0.5244, tangentMode=RCTM_User, time=0.97, value=0}
         ]}
        BlueCurve {keys [
          <at 0 0.1146 RCIM_Cubic>
          {arriveTangent=-0.1348, interpMode=RCIM_Cubic, leaveTangent=-0.1348, time=0.025, value=0.0419}
          <at 0.8503 0>
          <at 0.97 0>
         ]}
        GreenCurve {keys [
          <at 0 0.1464 RCIM_Cubic>
          {arriveTangent=-0.1463, interpMode=RCIM_Cubic, leaveTangent=-0.1463, time=0.025, value=0.0148}
          <at 0.8503 0.022>
          <at 0.97 0.0102>
         ]}
        RedCurve {keys [<at 0 6 RCIM_Cubic>,<at 0.025 1.5 RCIM_Cubic>,<at 0.8503 6>,<at 0.97 1.5>]}
       }})
     3=?
     Index=&MakeIntFromFloat(Float=$User.Variant, Conversion Mode=::NewEnumerator1)
     Num=::NewEnumerator1
    )
    Particles.RaysDynPar:Vector4f=&MakeVector4(
     X=&FloatFromCurve(FloatCurve=DataInterface() {propertyValues {Curve {keys [<at 0 0>,<at 1 1>]}}})
     Y=0
     Z=0
     W=0
    )
    Particles.RaysSpriteSize:Vector2f=&Vector2DFromFloat(Value=&FloatFromCurve(
      FloatCurve=DataInterface() {propertyValues {Curve {keys [
          {interpMode=RCIM_Cubic, leaveTangentWeight=0.0654, tangentMode=RCTM_User, time=0.7837, value=0.7714}
          {
           arriveTangent=2.6513
           arriveTangentWeight=0.1854
           interpMode=RCIM_Cubic
           leaveTangent=-0.002
           leaveTangentWeight=0.0033
           tangentMode=RCTM_Break
           time=0.9563
           value=1
          }
          {arriveTangentWeight=0.0033, interpMode=RCIM_Cubic, leaveTangentWeight=0.0017, tangentMode=RCTM_User, time=0.9663, value=1}
          {arriveTangentWeight=0.0017, interpMode=RCIM_Cubic, tangentMode=RCTM_User, time=0.995, value=0}
         ]}}}
      Scale Curve=&MaskFloatBySpawnGroup(Masked Float=150, Passthrough Float=120)
     ))
   ) {isSetParameters=true}
   &DynamicMaterialParameters(
    Index 0 Param 1=1
    Index 0 Param 2=1
    Index 0 Param 3=1
    Index 0 Param 4=2.3
    Index 2 Param 1=2.5
    Index 2 Param 3=1
    Index 2 Param 4=1
    Index 3 Param 1=30
    Module.Index0Float0Write=true
    Module.Index0Float1Write=true
    Module.Index0Float2Write=true
    Module.Index0Float3Write=true
    Module.Index1Float0Write=true
    Module.Index1Float1Write=true
    Module.Index1Float2Write=true
    Module.Index1Float3Write=true
    Module.Index2Float0Write=true
    Module.Index2Float1Write=true
    Module.Index2Float2Write=true
    Module.Index2Float3Write=true
    Module.Index3Float0Write=true
    Module.Index3Float1Write=true
    Module.Index3Float2Write=true
    Module.Index3Float3Write=true
    Write Parameter Index 1=::NewEnumerator1
    Write Parameter Index 2=::NewEnumerator1
    Write Parameter Index 3=::NewEnumerator1
   )
   /SetVariables(
    Particles.CenterColor:LinearColor=&MakeLinearColorFromVectorAndFloat(
     "Vector (RGB)"=&MakeVectorFromLinearColorRGB(Color=&MergeHelper_ColorSwitch(
       0=<rgba 1 0.332 0 1>
       1=<rgba 0 0.2952 1 1>
       2=<rgba 1 0.0059 0 1>
       3=?
       Index=&MakeIntFromFloat(Float=$User.Variant, Conversion Mode=::NewEnumerator1)
       Num=::NewEnumerator1
      ))
     "Float (Alpha)"=&Multiply_Float(
      A=&FNTimeOfDayScalar(Night Scale=0.5, Morning Scale=0.3, Evening Scale=0.4)
      B=&FloatFromCurve(FloatCurve=DataInterface() {propertyValues {Curve {keys [<at 0 0>,<at 0.0339 1>,<at 0.995 1>,<at 1 0>]}}})
     )
    )
    Particles.CenterSize:Vector2f=&Vector2DFromFloat(Value=&FloatFromCurve(FloatCurve=DataInterface() {propertyValues {Curve {keys [<at 0.9718 1.5 RCIM_Cubic>,<at 1 0>]}}}, Scale Curve=88.5783))
   ) {isSetParameters=true}
   &SkeletalMeshLocation(Bone Sampling Mode=::NewEnumerator4, MaskSpawnGroup=false, Skeletal Mesh=$System.Cubes)
  ]
  rendererClasses [&NiagaraSpriteRendererProperties]
  renderers [
   {rendererClass=&NiagaraSpriteRendererProperties}
   {rendererClass=&NiagaraSpriteRendererProperties, rendererIndex=1}
   {rendererClass=&NiagaraSpriteRendererProperties, rendererIndex=2}
  ]
 }
 Glitch {
  emitterUpdateScript [
   &EmitterState(Life Cycle Mode=::NewEnumerator1, Loop Behavior=::NewEnumerator1, Loop Duration=12, UseLoopDelay=true)
   &SpawnRate(
    SpawnGroup=1
    SpawnRate=&FloatFromCurve(
     FloatCurve=DataInterface() {propertyValues {Curve {
        defaultValue=1
        keys [<at 0 1 RCIM_Constant>, {arriveTangent=-42.7137, interpMode=RCIM_Constant, leaveTangent=-42.7137, time=0.0609, value=0}]
       }}}
     CurveIndex=$Emitter.NormalizedLoopAge
     Scale Curve=10
    )
    Use Spawn Probability=false
   )
   &SpawnBurst_Instantaneous(
    Spawn Count=3
    Spawn Time=&Add_Float(A=$System.Delay, B=9.42)
    Use Spawn Probability=false
    UseLoopCountLimit=false
   ) {module=SpawnBurst_Instantaneous002}
   &SpawnBurst_Instantaneous(
    Spawn Count=2
    Spawn Time=&Add_Float(A=$System.Delay, B=0.22)
    Use Spawn Probability=false
    UseLoopCountLimit=false
   ) {module=SpawnBurst_Instantaneous001}
   &SpawnBurst_Instantaneous(
    Spawn Count=3
    Spawn Time=&Add_Float(A=$System.Delay, B=0.16)
    Use Spawn Probability=false
    UseLoopCountLimit=false
   )
  ]
  particleSpawnScript [
   &InitializeParticle(
    Color=&MergeHelper_ColorSwitch(
     0=<rgba 5 1.0838 0 1>
     1=<rgba 0 2.6011 5 1>
     2=<rgba 5 0.1703 0.1649 1>
     3=?
     Index=&MakeIntFromFloat(Float=$User.Variant, Conversion Mode=::NewEnumerator1)
     Num=::NewEnumerator1
    )
    Lifetime=&MaskFloatBySpawnGroup(
     Masked Float=&RandomRangeFloat(Minimum=0.25, Maximum=0.5)
     Passthrough Float=&RandomRangeFloat(Minimum=0.25, Maximum=0.15)
    )
    Sprite Size=&RandomRangeVector2D(Minimum=<10 20>, Maximum=<15 30>)
    Sprite Size Mode=::NewEnumerator4
    Sprite UV Mode=::NewEnumerator1
    UsePositionOffset=false
    Write Material Random=true
   ) {module=InitializeParticle001}
   &CameraOffset(Camera Offset Amount=10)
   &SkeletalMeshLocation(
    Bone / Socket Index=&MaskIntBySpawnGroup(Masked Int=0, Passthrough Int=9, Spawn Group Mask=12)
    Bone Sampling Mode=::NewEnumerator4
    MaskSpawnGroup=false
    Position Sampling=::NewEnumerator2
    Skeletal Mesh=$System.Cubes
   )
   &ShapeLocation(
    MaskSpawnGroup=false
    Offset=<0 0 -10>
    Offset Mode=::NewEnumerator0
    Override Seed=false
    Shape Origin=&MaskVectorBySpawnGroup(
     Masked Vector=Linked() {linkedVariable {name=Particles.SkeletalMeshLocation.SampledPosition, type=NiagaraPosition}}
     Passthrough Vector=&AddVectorToPosition(Position=$Particles.SkeletalMeshLocation.SampledPosition, Vector=<0 0 0>)
    )
    Sphere Radius=10
    UseOwnerScale=true
   )
  ]
  particleUpdateScript [
   &ParticleState()
   &DynamicMaterialParameters(
    Index 0 Param 1=&RandomRangeFloat(Minimum=0, Maximum=1)
    Index 0 Param 2=&Lerp_Float(A=0, B=0.5, Alpha=$Particles.NormalizedAge)
    Index 0 Param 3=/Niagara/DynamicInputs/Integer/V2/Round_Float(Float:NiagaraFloat=&RandomRangeFloat(Minimum=0, Maximum=3))
    Module.Index0Float0Write=true
    Module.Index0Float1Write=true
    Module.Index0Float2Write=true
    Module.Index0Float3Write=true
   )
   &ScaleSpriteSize(
    Scale Factor=&Vector2DFromCurve(
     Vector2Curve=DataInterface() {propertyValues {
       XCurve {keys [<at 0 1>,<at 1 1>]}
       YCurve {keys [{arriveTangent=4.2107, interpMode=RCIM_Cubic, leaveTangent=4.2107, tangentMode=RCTM_User, time=0, value=0.5}, <at 1 2 RCIM_Cubic>]}
      }}
     ScaleCurve=<1 1>
    )
    Scale Sprite Size Mode=::NewEnumerator2
   )
   &ScaleColor(
    Scale Alpha=&FloatFromCurve(FloatCurve=DataInterface() {propertyValues {Curve {keys [
         {arriveTangent=3.8564, interpMode=RCIM_Cubic, leaveTangent=3.8564, tangentMode=RCTM_User, time=0, value=0.5274}
         {arriveTangent=0.1295, interpMode=RCIM_Cubic, leaveTangent=0.1295, tangentMode=RCTM_User, time=0.5, value=1}
         <at 1 0 RCIM_Cubic>
        ]}}})
    ScaleA=true
    ScaleRGB=false
   )
  ]
  rendererClasses [&NiagaraSpriteRendererProperties]
  renderers [{rendererClass=&NiagaraSpriteRendererProperties}]
 }
 Glow {
  bEnabled=false
  emitterUpdateScript [
   &EmitterState(
    Life Cycle Mode=::NewEnumerator1
    Loop Behavior=::NewEnumerator1
    Loop Delay=&Add_Float(A=$System.Delay, B=2)
    Loop Duration=10.8
    UseLoopDelay=true
   )
   &SpawnBurst_Instantaneous(Use Spawn Probability=false, UseLoopCountLimit=false)
  ]
  particleSpawnScript [&InitializeParticle(
    Color=&SelectLinearColorFromArray(
     Array Sampling Mode=::NewEnumerator1
     Direct Array Index=0
     Color Selection Array=DataInterface() {propertyValues {ColorData [<rgba 1 0.3784 0 1>, <rgba 0 0.4916 1 1>, <rgba 1 0.3784 0 1>]}}
     Random Seed=?
    )
    Lifetime=8
    Sprite Rotation Mode=::NewEnumerator2
    Sprite Size Mode=::NewEnumerator3
    Uniform Sprite Size=120
    UsePositionOffset=false
    Write Material Random=true
   )]
  particleUpdateScript [
   &ParticleState()
   &SpriteRotationRate(Rotation Rate=&FloatFromCurve(
     FloatCurve=DataInterface() {propertyValues {Curve {
        defaultValue=1
        keys [
         {interpMode=RCIM_Cubic, tangentMode=RCTM_User, time=0, value=0}
         {arriveTangent=-4.4276, interpMode=RCIM_Cubic, leaveTangent=-4.4276, time=0.0958, value=1.1104}
         {arriveTangentWeight=0.1138, interpMode=RCIM_Cubic, leaveTangentWeight=0.2881, tangentMode=RCTM_User, time=0.2911, value=-1.2889}
         {arriveTangentWeight=0.2866, interpMode=RCIM_Cubic, leaveTangentWeight=0.5185, tangentMode=RCTM_User, time=0.6308, value=0}
         <at 0.7688 0.18 RCIM_Cubic>
         {interpMode=RCIM_Cubic, tangentMode=RCTM_User, time=1, value=0}
        ]
       }}}
     Scale Curve=400
    ))
   &ScaleColor(
    Scale Alpha=&FloatFromCurve(FloatCurve=DataInterface() {propertyValues {Curve {keys [
         <at 0 0.5 RCIM_Cubic>
         {arriveTangentWeight=0.0865, interpMode=RCIM_Cubic, leaveTangentWeight=0.2862, tangentMode=RCTM_User, time=0.2321, value=1}
         {arriveTangentWeight=0.2434, interpMode=RCIM_Cubic, leaveTangentWeight=0.1554, tangentMode=RCTM_User, time=0.985, value=1}
        ]}}})
    Scale RGB=&VectorFromFloat(Value=&FloatFromCurve(FloatCurve=DataInterface() {propertyValues {Curve {keys [
          {interpMode=RCIM_Cubic, tangentMode=RCTM_User, time=0, value=0.5}
          <at 0.5 1 RCIM_Cubic>
          {arriveTangent=-0.1284, interpMode=RCIM_Cubic, leaveTangent=-0.1284, tangentMode=RCTM_User, time=1, value=0.5}
         ]}}}))
    ScaleA=true
    ScaleRGB=true
   )
   &ScaleSpriteSize(
    Scale Factor=&Vector2DFromFloat(Value=&FloatFromCurve(FloatCurve=DataInterface() {propertyValues {Curve {keys [
          {interpMode=RCIM_Cubic, tangentMode=RCTM_User, time=0, value=0.5}
          {
           arriveTangent=0.9101
           arriveTangentWeight=0.1572
           interpMode=RCIM_Cubic
           leaveTangent=0.9101
           leaveTangentWeight=0.187
           tangentMode=RCTM_User
           time=0.2013
           value=0.7311
          }
          {arriveTangentWeight=0.0486, interpMode=RCIM_Cubic, leaveTangentWeight=0.0063, tangentMode=RCTM_User, time=0.2316, value=0.6814}
          {arriveTangentWeight=0.0022, interpMode=RCIM_Cubic, leaveTangentWeight=0.0079, tangentMode=RCTM_User, time=0.2582, value=0.8592}
          {arriveTangentWeight=0.0079, interpMode=RCIM_Cubic, leaveTangentWeight=0.005, tangentMode=RCTM_User, time=0.2794, value=0.7085}
          {arriveTangentWeight=0.0056, interpMode=RCIM_Cubic, leaveTangentWeight=0.0103, tangentMode=RCTM_User, time=0.2947, value=1.0562}
          {arriveTangentWeight=0.0098, interpMode=RCIM_Cubic, leaveTangentWeight=0.0053, tangentMode=RCTM_User, time=0.3173, value=0.8333}
          {arriveTangentWeight=0.0153, interpMode=RCIM_Cubic, leaveTangentWeight=0.0391, tangentMode=RCTM_User, time=0.3257, value=1.2148}
          {arriveTangentWeight=0.2165, interpMode=RCIM_Cubic, leaveTangentWeight=0.1618, tangentMode=RCTM_User, time=0.3479, value=1.328}
          {arriveTangentWeight=0.1546, interpMode=RCIM_Cubic, leaveTangentWeight=0.0794, tangentMode=RCTM_User, time=0.3599, value=0.9473}
          {arriveTangentWeight=0.0037, interpMode=RCIM_Cubic, leaveTangentWeight=0.003, tangentMode=RCTM_User, time=0.3772, value=1.2876}
          {arriveTangentWeight=0.003, interpMode=RCIM_Cubic, leaveTangentWeight=0.0044, tangentMode=RCTM_User, time=0.3892, value=0.9119}
          {arriveTangentWeight=0.0044, interpMode=RCIM_Cubic, leaveTangentWeight=0.0039, tangentMode=RCTM_User, time=0.3981, value=1.849}
          {arriveTangentWeight=0.0039, interpMode=RCIM_Cubic, leaveTangentWeight=0.0028, tangentMode=RCTM_User, time=0.4083, value=1.285}
          {arriveTangentWeight=0.0028, interpMode=RCIM_Cubic, leaveTangentWeight=0.0048, tangentMode=RCTM_User, time=0.4181, value=1.7376}
          {arriveTangentWeight=0.0048, interpMode=RCIM_Cubic, leaveTangentWeight=0.0024, tangentMode=RCTM_User, time=0.4338, value=1.4407}
          {arriveTangentWeight=0.146, interpMode=RCIM_Cubic, leaveTangentWeight=0.1771, tangentMode=RCTM_User, time=0.4486, value=1.9071}
          {arriveTangentWeight=0.0029, interpMode=RCIM_Cubic, leaveTangentWeight=0.0043, tangentMode=RCTM_User, time=0.4668, value=1.3099}
          {arriveTangentWeight=0.014, interpMode=RCIM_Cubic, leaveTangentWeight=0.0241, tangentMode=RCTM_User, time=0.4777, value=1.9292}
          {arriveTangentWeight=0.0505, interpMode=RCIM_Cubic, leaveTangentWeight=0.0235, tangentMode=RCTM_User, time=0.5033, value=2.2433}
          {arriveTangentWeight=0.0186, interpMode=RCIM_Cubic, leaveTangentWeight=0.0293, tangentMode=RCTM_User, time=0.5163, value=0.896}
          {arriveTangentWeight=0.0007, interpMode=RCIM_Cubic, leaveTangentWeight=0.1139, tangentMode=RCTM_User, time=0.525, value=2.2846}
          {arriveTangentWeight=0.0104, interpMode=RCIM_Cubic, leaveTangentWeight=0.1242, tangentMode=RCTM_User, time=0.5633, value=1.7}
          {arriveTangent=0.5015, interpMode=RCIM_Cubic, leaveTangent=0.5015, time=0.6173, value=1.6127}
          {arriveTangent=-0.0268, interpMode=RCIM_Cubic, leaveTangent=-0.0268, time=0.7032, value=1.7701}
          {arriveTangent=-0.3546, interpMode=RCIM_Cubic, leaveTangent=-0.3546, time=0.8235, value=1.6072}
          {interpMode=RCIM_Cubic, tangentMode=RCTM_User, time=0.901, value=1.7}
          {arriveTangentWeight=0.0737, interpMode=RCIM_Cubic, leaveTangentWeight=0.0932, tangentMode=RCTM_User, time=0.92, value=2.2798}
          {arriveTangentWeight=0.6437, interpMode=RCIM_Cubic, tangentMode=RCTM_User, time=0.935, value=0}
         ]}}}))
    Scale Sprite Size Mode=::NewEnumerator2
   )
   &SkeletalMeshLocation(Bone Sampling Mode=::NewEnumerator4, MaskSpawnGroup=false, Skeletal Mesh=$System.Cubes)
   &Light_Attributes(
    Color=1
    Color Value To Scale:LinearColor=&Multiply_LinearColorByFloat(
     Linear Color=&SelectLinearColorFromArray(
      Array Sampling Mode=::NewEnumerator1
      Direct Array Index=0
      Color Selection Array=DataInterface() {propertyValues {ColorData [<rgba 20 9.2925 0 1>, <rgba 0 8.3625 20 1>, <rgba 20 9.2925 0 1>]}}
      Random Seed=?
     )
     Float=1
    )
    Light Radius=170
    Radius=1
    Scale Alpha:NiagaraFloat=&FloatFromCurve(FloatCurve=DataInterface() {propertyValues {Curve {
        defaultValue=1
        keys [
         {arriveTangent=0.1755, interpMode=RCIM_Cubic, leaveTangent=0.1755, tangentMode=RCTM_User, time=0, value=0}
         {
          arriveTangent=0.8794
          arriveTangentWeight=0.1038
          interpMode=RCIM_Cubic
          leaveTangent=0.8794
          leaveTangentWeight=0.3962
          tangentMode=RCTM_User
          time=0.5081
          value=0.2389
         }
         {arriveTangent=16.759, interpMode=RCIM_Cubic, leaveTangent=16.759, time=0.5244, value=2.1742}
         {arriveTangent=-3.0713, interpMode=RCIM_Cubic, leaveTangent=-3.0713, time=0.5327, value=0.6514}
         {arriveTangentWeight=0.1897, interpMode=RCIM_Cubic, leaveTangentWeight=0.1744, tangentMode=RCTM_User, time=0.539, value=2.1295}
         {arriveTangentWeight=0.0372, interpMode=RCIM_Cubic, leaveTangentWeight=0.2083, tangentMode=RCTM_User, time=0.5464, value=1}
         {arriveTangent=-4.2228, interpMode=RCIM_Cubic, leaveTangent=-4.2228, tangentMode=RCTM_User, time=0.9477, value=0.5527}
         {interpMode=RCIM_Cubic, leaveTangent=-1.1146, time=0.9499, value=0}
        ]
       }}})
    ScaleA=true
    ScaleRGB=false
   )
  ]
  rendererClasses [&NiagaraSpriteRendererProperties, &NiagaraLightRendererProperties]
  renderers [
   {rendererClass=&NiagaraSpriteRendererProperties}
   {rendererClass=&NiagaraLightRendererProperties, rendererIndex=1}
  ]
 }
 Hunt_Lightning {
  bEnabled=false
  emitterUpdateScript [
   &EmitterState(
    Life Cycle Mode=::NewEnumerator1
    Loop Behavior=::NewEnumerator1
    Loop Delay=&Add_Float(A=$System.Delay, B=1.5)
    Loop Duration=4.7
    UseLoopDelay=true
   )
   &SpawnRate(
    SpawnRate=&FloatFromCurve(
     FloatCurve=DataInterface() {propertyValues {Curve {
        defaultValue=1
        keys [
         {interpMode=RCIM_Cubic, tangentMode=RCTM_User, time=0, value=0}
         {arriveTangent=1.6185, interpMode=RCIM_Cubic, leaveTangent=1.6185, time=0.3782, value=0.3348}
         {arriveTangent=3.0969, interpMode=RCIM_Cubic, leaveTangent=3.0969, tangentMode=RCTM_User, time=0.6179, value=1}
        ]
       }}}
     CurveIndex=$Emitter.NormalizedLoopAge
     Scale Curve=30
    )
    Use Spawn Probability=false
   )
  ]
  particleSpawnScript [
   &InitializeParticle(
    Color=&MergeHelper_ColorSwitch(0=<rgba 5 3.6617 0 1>, 1=<rgba 0 3.1555 5 1>, 2=<rgba 5 3.6617 0 1>, 3=?, Index=0, Num=::NewEnumerator1)
    Lifetime Max=&FloatFromCurve(FloatCurve=DataInterface() {propertyValues {Curve {keys [<at 0.2327 0.5>,<at 0.6279 0.15>]}}}, CurveIndex=$Emitter.NormalizedLoopAge)
    Lifetime Min=0.25
    Lifetime Mode=::NewEnumerator1
    Mesh Scale=&Multiply_VectorByFloat(
     Vector=&Lerp_Vector(
      A=&RandomRangeVector(Minimum=<0.5 0.4 0.4>, Maximum=<0.75 0.6 0.6>)
      B=&RandomRangeVector(Minimum=<1 1 1>, Maximum=<1.5 2 2>)
      Alpha=&FloatFromCurve(
       FloatCurve=DataInterface() {propertyValues {Curve {keys [{interpMode=RCIM_Cubic, tangentMode=RCTM_User, time=0.5, value=0}, <at 1 1>]}}}
       CurveIndex=&RandomRangeFloat(Minimum=0, Maximum=1)
      )
     )
     Float=0.6
    )
    Mesh Scale Mode=::NewEnumerator4
    UsePositionOffset=false
    Write Material Random=true
   )
   /SetVariables(Particles.Random=&RandomRangeFloat(Minimum=0, Maximum=1)) {isSetParameters=true}
   &AddVelocity(Module.Use Falloff Distance=false, Override Seed=false, Velocity Mode=::NewEnumerator1, Velocity Speed=-20)
  ]
  particleUpdateScript [
   &ParticleState()
   &ScaleMeshSize(Scale Factor=&VectorFromCurve(VectorCurve=DataInterface() {propertyValues {
       XCurve {keys [<at 0.2848 1 RCIM_Cubic>, {arriveTangent=-3.0417, interpMode=RCIM_Cubic, leaveTangent=-3.0417, tangentMode=RCTM_User, time=1, value=0}]}
       YCurve {keys [<at 0 1>,<at 1 1>]}
       ZCurve {keys [<at 0 1>,<at 1 1>]}
      }}))
   &ScaleColor(
    Scale Alpha=&Multiply_Float(
     A=10
     B=&FloatFromCurve(FloatCurve=DataInterface() {propertyValues {Curve {keys [
          {interpMode=RCIM_Cubic, tangentMode=RCTM_User, time=0, value=0}
          {interpMode=RCIM_Cubic, tangentMode=RCTM_User, time=0.1, value=0}
          {interpMode=RCIM_Cubic, tangentMode=RCTM_User, time=0.2, value=0.8}
          {interpMode=RCIM_Cubic, tangentMode=RCTM_User, time=1, value=0}
         ]}}})
    )
    Scale RGB=&Multiply_VectorByFloat(
     Vector=&Multiply_VectorByFloat(Vector=<1 1 1>, Float=&FN_FrontEndFloatModifier(Gameplay=1, Front End=0.5))
     Float=&Multiply_Float(
      A=&RandomRangeFloat(Minimum=0.5, Maximum=1, Evaluation Type=::NewEnumerator1)
      B=&Multiply_Float(A=&FNTimeOfDayScalar(Night Scale=0.2, Morning Scale=1, Evening Scale=1), B=10)
     )
    )
    ScaleA=true
    ScaleRGB=true
   )
   &DynamicMaterialParameters(
    Index 0 Param 1=&FloatFromCurve(FloatCurve=DataInterface() {propertyValues {Curve {keys [{tangentMode=RCTM_User, time=0, value=1}, {tangentMode=RCTM_User, time=1, value=1}]}}})
    Index 0 Param 2=0.7
    Index 0 Param 3=&FloatFromCurve(FloatCurve=DataInterface() {propertyValues {Curve {keys [
         {arriveTangent=4.0072, interpMode=RCIM_Cubic, leaveTangent=4.0072, tangentMode=RCTM_User, time=0, value=0}
         {arriveTangent=2.7083, interpMode=RCIM_Cubic, leaveTangent=2.7083, time=0.2543, value=0.6225}
         {arriveTangent=1.8473, interpMode=RCIM_Cubic, leaveTangent=1.8473, time=0.6, value=1.625}
         {arriveTangent=1.6993, interpMode=RCIM_Cubic, leaveTangent=1.6993, tangentMode=RCTM_User, time=1, value=2}
        ]}}})
    Index 0 Param 4=5
    Index 1 Param 1=0.7
    Index 1 Param 2=$Particles.Random
    Index 1 Param 3=1.3
    Index 1 Param 4=1
    Index 2 Param 1=&FloatFromCurve(FloatCurve=DataInterface() {propertyValues {Curve {keys [<at 0 1>,<at 1 0.3>]}}})
    Module.Index0Float0Write=true
    Module.Index0Float1Write=true
    Module.Index0Float2Write=true
    Module.Index0Float3Write=true
    Module.Index1Float0Write=true
    Module.Index1Float1Write=true
    Module.Index1Float2Write=true
    Module.Index1Float3Write=true
    Module.Index2Float0Write=true
    Module.Index2Float1Write=true
    Module.Index2Float2Write=true
    Module.Index2Float3Write=true
    Write Parameter Index 1=::NewEnumerator1
    Write Parameter Index 2=::NewEnumerator1
   )
   &SolveForcesAndVelocity(Clamp Velocity=false, Limit Acceleration=false)
   /SetVariables(Particles.Lightning glow color:LinearColor=&MakeLinearColorFromVectorAndFloat(
     "Vector (RGB)"=&MakeVectorFromLinearColorRGB(Color=&MultiplyLinearColor(
       Linear Color=&SelectLinearColorFromArray(
        Array Sampling Mode=::NewEnumerator1
        Direct Array Index=0
        Color Selection Array=DataInterface() {propertyValues {ColorData [<rgba 4 1.6071 0.1968 1>, <rgba 0.3366 2.0701 4 1>, <rgba 4 1.6071 0.1968 1>]}}
        Random Seed=?
       )
       Scale Factor=&MakeVector4_FromFloat(Float=&RandomRangeFloat(Minimum=0.5, Maximum=1, Evaluation Type=::NewEnumerator1))
      ))
     "Float (Alpha)"=&Multiply_Float(
      A=&FloatFromCurve(FloatCurve=DataInterface() {propertyValues {Curve {keys [<at 0 1 RCIM_Cubic>, {interpMode=RCIM_Cubic, leaveTangent=-3.6307, time=0.9, value=0}]}}})
      B=1
     )
    )) {isSetParameters=true}
   /SetVariables(Particles.Glow Dy02:Vector4f=&MakeVector4(
     X=&FloatFromCurve(FloatCurve=DataInterface() {propertyValues {Curve {keys [{tangentMode=RCTM_User, time=0, value=0.2}, {tangentMode=RCTM_User, time=0.5, value=1}]}}})
     Y=0.7
     Z=0
     W=5
    )) {isSetParameters=true}
   /SetVariables(Particles.Glow Dy03:Vector4f=&MakeVector4(
     X=&FloatFromCurve(
      FloatCurve=DataInterface() {propertyValues {Curve {keys [<at 0 1>,<at 1 0.5>]}}}
      Scale Curve=&MaskFloat(Mask Source=1, Mask Parameter=0, Mask=2, Masked Float=0.7, Passthrough Float=0.5)
     )
     Y=$Particles.Random
     Z=1.3
     W=1
    )) {isSetParameters=true}
   &SkeletalMeshLocation(Bone Sampling Mode=::NewEnumerator4, MaskSpawnGroup=false, Skeletal Mesh=$System.Cubes)
   &ShapeLocation(
    MaskSpawnGroup=false
    Override Seed=false
    Shape Origin=$Particles.SkeletalMeshLocation.SampledPosition
    Sphere Radius=10
    Sphere Surface Distribution=1
    UseOwnerScale=true
   )
  ]
  rendererClasses [&NiagaraMeshRendererProperties]
  renderers [
   {rendererClass=&NiagaraMeshRendererProperties, rendererIndex=0}
   {rendererClass=&NiagaraMeshRendererProperties, rendererIndex=1}
  ]
 }
 Lightning {
  emitterUpdateScript [
   &EmitterState(
    Inactive Response=::NewEnumerator1
    Life Cycle Mode=::NewEnumerator1
    Loop Behavior=::NewEnumerator1
    Loop Delay=$System.Delay
    Loop Duration=9
    UseLoopDelay=true
   )
   &SpawnRate(
    SpawnRate=&FloatFromCurve(
     FloatCurve=DataInterface() {propertyValues {Curve {keys [
         {interpMode=RCIM_Cubic, tangentMode=RCTM_User, time=0.1032, value=0}
         {arriveTangent=1.2782, interpMode=RCIM_Cubic, leaveTangent=1.2782, tangentMode=RCTM_User, time=0.5021, value=0.1471}
         {arriveTangent=6.3755, interpMode=RCIM_Cubic, leaveTangent=6.3755, time=0.6776, value=0.8504}
         {arriveTangent=9.9865, interpMode=RCIM_Cubic, leaveTangent=9.9865, tangentMode=RCTM_User, time=0.7269, value=1.5805}
        ]}}}
     CurveIndex=$Emitter.NormalizedLoopAge
     Scale Curve=50
    )
    Use Spawn Probability=false
   )
  ]
  particleSpawnScript [
   &PartitionParticles(Number Of Partitions=11)
   &InitializeParticle(
    Color Mode=::NewEnumerator0
    Lifetime Max=0.25
    Lifetime Min=0.15
    Lifetime Mode=::NewEnumerator1
    Mesh Scale=&Multiply_VectorByFloat(Vector=&RandomRangeVector(Minimum=<0.5 0.9 0.9>, Maximum=<0.5 0.7311 0.6712>), Float=1.2)
    Mesh Scale Mode=::NewEnumerator4
    UsePositionOffset=false
    Write Material Random=true
   )
   &SkeletalMeshLocation(
    Bone / Socket Index=$Particles.PartitionParticles.Partition
    Bone Sampling Mode=::NewEnumerator4
    MaskSpawnGroup=false
    Skeletal Mesh=$System.Cubes
   )
   &ShapeLocation(
    Handle Radius=0
    Large Radius=10
    MaskSpawnGroup=false
    Offset=<0 10 0>
    Offset Mode=::NewEnumerator0
    Override Seed=false
    Rotation Mode=::NewEnumerator2
    Shape Origin=$Particles.SkeletalMeshLocation.SampledPosition
    Shape Primitive=::NewEnumerator3
    Surface Distribution=1
    UseOwnerScale=true
    V Distribution=1
    Yaw / Pitch / Roll=<0 0 0.25>
   )
   &AddVelocity(
    Module.Use Falloff Distance=false
    Origin Offset=<0 10 0>
    Origin Offset Coordinate Space=::Simulation
    Override Seed=false
    Velocity Mode=::NewEnumerator1
    Velocity Origin=$Particles.SkeletalMeshLocation.SampledPosition
    Velocity Speed=0.1
   )
   &InitialMeshOrientation(Mesh Orientation Mode=::NewEnumerator2, Orientation Vector=$StackContext.Velocity, Rotation=<0.2519 0 0>, Use Rotation Vector=true) {module=InitialMeshOrientation001}
  ]
  particleUpdateScript [
   &ParticleState()
   &DynamicMaterialParameters(
    Index 0 Param 1=&MakeFloatFromVector(Vector=$Particles.Scale, Channel=::NewEnumerator0)
    Index 0 Param 2=&RandomRangeFloat(Minimum=3, Maximum=10)
    Index 0 Param 3=&FloatFromCurve(
     FloatCurve=DataInterface() {propertyValues {Curve {keys [{interpMode=RCIM_Cubic, leaveTangent=8.1085, tangentMode=RCTM_Break, time=0.8568, value=1}, {tangentMode=RCTM_User, time=1, value=0}]}}}
     Scale Curve=0.005
    )
    Index 0 Param 4=&RandomRangeFloat(Minimum=0.4, Maximum=0.2)
    Index 1 Param 1=&FloatFromCurve(FloatCurve=DataInterface() {propertyValues {Curve {keys [
         {interpMode=RCIM_Cubic, leaveTangent=1.1034, tangentMode=RCTM_Break, time=0, value=0}
         {arriveTangent=2.2399, tangentMode=RCTM_User, time=0.5, value=1}
        ]}}})
    Index 1 Param 2=&FloatFromCurve(FloatCurve=DataInterface() {propertyValues {Curve {keys [<at 0.5 0>, {arriveTangent=3.2053, tangentMode=RCTM_User, time=0.9, value=1}]}}})
    Index 1 Param 3=&FloatFromCurve(FloatCurve=DataInterface() {propertyValues {Curve {keys [
         <at 0 0 RCIM_Cubic>
         {arriveTangent=0.5046, interpMode=RCIM_Cubic, leaveTangent=0.5046, time=0.8846, value=0.0963}
         {arriveTangent=2.773, interpMode=RCIM_Cubic, leaveTangent=2.773, tangentMode=RCTM_User, time=1, value=0.5046}
        ]}}})
    Index 1 Param 4=&FloatFromCurve(
     FloatCurve=DataInterface() {propertyValues {Curve {keys [
         {interpMode=RCIM_Cubic, leaveTangent=0.1169, tangentMode=RCTM_Break, time=0, value=0}
         {arriveTangent=2.5708, interpMode=RCIM_Cubic, leaveTangent=2.5708, tangentMode=RCTM_User, time=1, value=1}
        ]}}}
     Scale Curve=&RandomRangeFloat(Minimum=1, Maximum=40)
    )
    Index 2 Param 1=&FloatFromCurve(FloatCurve=DataInterface() {propertyValues {Curve {keys [
         {interpMode=RCIM_Cubic, leaveTangent=-0.0025, tangentMode=RCTM_Break, time=0, value=0.5}
         {arriveTangent=-0.0316, interpMode=RCIM_Cubic, tangentMode=RCTM_Break, time=1, value=-0.2}
        ]}}})
    Index 2 Param 2=&FloatFromCurve(
     FloatCurve=DataInterface() {propertyValues {Curve {keys [
         {interpMode=RCIM_Cubic, tangentMode=RCTM_User, time=0, value=0}
         {arriveTangent=2.0522, interpMode=RCIM_Cubic, leaveTangent=2.0522, tangentMode=RCTM_User, time=1, value=1}
        ]}}}
     Scale Curve=&RandomRangeFloat(Minimum=0.4, Maximum=1)
    )
    Module.Index0Float0Write=true
    Module.Index0Float1Write=true
    Module.Index0Float2Write=true
    Module.Index0Float3Write=true
    Module.Index1Float0Write=true
    Module.Index1Float1Write=true
    Module.Index1Float2Write=true
    Module.Index1Float3Write=true
    Module.Index2Float0Write=true
    Module.Index2Float1Write=true
    Module.Index2Float2Write=false
    Module.Index2Float3Write=false
    Module.Index3Float0Write=false
    Module.Index3Float1Write=false
    Module.Index3Float2Write=false
    Module.Index3Float3Write=true
    Write Parameter Index 1=::NewEnumerator1
    Write Parameter Index 2=::NewEnumerator1
    Write Parameter Index 3=::NewEnumerator1
   )
   &SolveForcesAndVelocity(Clamp Velocity=false, Limit Acceleration=false)
   &ScaleColor(
    Scale Alpha=&Multiply_Float(
     A=20
     B=&FloatFromCurve(FloatCurve=DataInterface() {propertyValues {Curve {keys [
          {tangentMode=RCTM_User, time=0, value=1}
          {arriveTangent=-0.0405, interpMode=RCIM_Cubic, leaveTangent=-0.0405, tangentMode=RCTM_User, time=0.7009, value=1}
          {arriveTangent=-8.0089, interpMode=RCIM_Cubic, leaveTangent=-8.0089, tangentMode=RCTM_User, time=1, value=0}
         ]}}})
    )
    Scale RGB=&MakeVectorFromLinearColorRGB(Color=&MergeHelper_ColorSwitch(
      0=&ColorFromCurve(ColorCurve=DataInterface() {propertyValues {
         AlphaCurve {keys [<at 0 1 RCIM_Cubic>]}
         BlueCurve {keys [<at 0 0>,<at 0.4295 0 RCIM_Cubic>,<at 0.5121 0>,<at 0.8639 0>,<at 1 0 RCIM_Cubic>]}
         GreenCurve {keys [
           <at 0 3.6617>
           {arriveTangent=-5.8859, interpMode=RCIM_Cubic, leaveTangent=-5.8859, time=0.4295, value=0.7323}
           <at 0.5121 0.6473>
           <at 0.8639 0.3486>
           <at 1 0.1206 RCIM_Cubic>
          ]}
         RedCurve {keys [
           <at 0 5>
           {arriveTangent=-7.8104, interpMode=RCIM_Cubic, leaveTangent=-7.8104, time=0.4295, value=1}
           <at 0.5121 1>
           <at 0.8639 1>
           <at 1 1 RCIM_Cubic>
          ]}
        }})
      1=&ColorFromCurve(ColorCurve=DataInterface() {propertyValues {
         AlphaCurve {keys [<at 0 1 RCIM_Cubic>]}
         BlueCurve {keys [
           <at 0 5>
           {arriveTangent=-7.8104, interpMode=RCIM_Cubic, leaveTangent=-7.8104, time=0.4295, value=1}
           <at 0.5121 1>
           <at 0.8639 1>
           <at 1 1 RCIM_Cubic>
          ]}
         GreenCurve {keys [
           <at 0 2.9824>
           {arriveTangent=-4.3283, interpMode=RCIM_Cubic, leaveTangent=-4.3283, time=0.4295, value=0.5386}
           <at 0.5121 0.7657>
           <at 0.8639 0.2961>
           <at 1 0.1481 RCIM_Cubic>
          ]}
         RedCurve {keys [<at 0 0>,<at 0.4295 0 RCIM_Cubic>,<at 0.5121 0>,<at 0.8639 0>,<at 1 0 RCIM_Cubic>]}
        }})
      2=&ColorFromCurve(ColorCurve=DataInterface() {propertyValues {
         AlphaCurve {keys [<at 0 1 RCIM_Cubic>]}
         BlueCurve {keys [
           <at 0 0.5788>
           {arriveTangent=-1.0541, interpMode=RCIM_Cubic, leaveTangent=-1.0541, time=0.4295, value=0.0115}
           <at 0.5121 0.0389>
           <at 0.8639 0>
           <at 1 0 RCIM_Cubic>
          ]}
         GreenCurve {keys [
           <at 0 0.8264>
           {arriveTangent=-1.6137, interpMode=RCIM_Cubic, leaveTangent=-1.6137, time=0.4295, value=0}
           <at 0.5121 0>
           <at 0.8639 0.02>
           <at 1 0.0059 RCIM_Cubic>
          ]}
         RedCurve {keys [
           <at 0 5>
           {arriveTangent=-7.8104, interpMode=RCIM_Cubic, leaveTangent=-7.8104, time=0.4295, value=1}
           <at 0.5121 1>
           <at 0.8639 1>
           <at 1 1 RCIM_Cubic>
          ]}
        }})
      3=?
      Index=&MakeIntFromFloat(Float=$User.Variant, Conversion Mode=::NewEnumerator1)
      Num=::NewEnumerator1
     ))
    ScaleA=true
    ScaleRGB=true
   )
   &ScaleMeshSize(Scale Factor=&Lerp_Vector(
     A=$Particles.Initial.Scale
     B=&Multiply_VectorByFloat(Vector=$Particles.SkeletalMeshLocation.SampledScale, Float=0.5)
     Alpha=&FloatFromCurve(
      FloatCurve=DataInterface() {propertyValues {Curve {defaultValue=1, keys [<at 0 0 RCIM_Constant>,<at 0.7001 1 RCIM_Constant>]}}}
      CurveIndex=$Emitter.NormalizedLoopAge
     )
    ))
  ]
  rendererClasses [&NiagaraMeshRendererProperties]
  renderers [
   {rendererClass=&NiagaraMeshRendererProperties, rendererIndex=0}
   {rendererClass=&NiagaraMeshRendererProperties, rendererIndex=1}
  ]
 }
 Runes {
  emitterUpdateScript [
   &EmitterState(
    Life Cycle Mode=::NewEnumerator1
    Loop Behavior=::NewEnumerator1
    Loop Delay=&Add_Float(A=$System.Delay, B=4)
    UseLoopDelay=true
   )
   &SpawnBurst_Instantaneous(Spawn Count=10, Spawn Time=2.25, Use Spawn Probability=false, UseLoopCountLimit=false)
   &SpawnRate(
    SpawnRate=&FloatFromCurve(
     FloatCurve=DataInterface() {propertyValues {Curve {keys [
         <at 0.438 1 RCIM_Cubic>
         {arriveTangentWeight=0.0865, interpMode=RCIM_Cubic, leaveTangentWeight=0.2862, tangentMode=RCTM_User, time=0.56, value=1}
         {arriveTangent=0.4151, interpMode=RCIM_Cubic, leaveTangent=0.4151, tangentMode=RCTM_User, time=0.5744, value=0.0605}
         {arriveTangent=4.1556, interpMode=RCIM_Cubic, leaveTangent=4.1556, time=0.6549, value=0.2477}
         {arriveTangent=0.3292, interpMode=RCIM_Cubic, leaveTangent=0.3292, time=0.7335, value=0.7215}
         {
          arriveTangent=-0.8585
          arriveTangentWeight=0.2434
          interpMode=RCIM_Cubic
          leaveTangent=-0.8585
          leaveTangentWeight=0.1554
          tangentMode=RCTM_User
          time=0.8963
          value=0.3272
         }
        ]}}}
     CurveIndex=$Emitter.NormalizedLoopAge
     Scale Curve=6
    )
    Use Spawn Probability=true
   )
  ]
  particleSpawnScript [
   /SetVariables(Particles.VisibilityTag=&MakeIntFromFloat(Float=$User.Variant, Conversion Mode=::NewEnumerator1)) {isSetParameters=true}
   &InitializeParticle(
    Lifetime=&RandomRangeFloat(Minimum=0.5, Maximum=2)
    Sprite Size Mode=::NewEnumerator1
    Uniform Sprite Size Max=3
    Uniform Sprite Size Min=7
    UsePositionOffset=false
    Write Material Random=true
   )
   &SkeletalMeshLocation(
    Bone Sampling Mode=::NewEnumerator4
    MaskSpawnGroup=false
    Sampled Scale Coordinate Space:ENiagaraCoordinateSpace=::World
    Skeletal Mesh=$System.Cubes
   )
   &ShapeLocation(MaskSpawnGroup=false, Offset Mode=::NewEnumerator0, Override Seed=false, Sphere Radius=45, UseOwnerScale=true)
   &AddVelocity(
    Module.Use Falloff Distance=false
    Override Seed=false
    Velocity Mode=::NewEnumerator1
    Velocity Origin=$Particles.SkeletalMeshLocation.SampledPosition
    Velocity Speed=&RandomRangeFloat(Minimum=100, Maximum=200)
   )
  ]
  particleUpdateScript [
   &ParticleState()
   &DynamicMaterialParameters(
    Index 0 Param 1=&Waveform(
     "Waveform [1]"=::NewEnumerator9
     "[1] Phase Scale"=&RandomRangeFloat(Minimum=0, Maximum=1)
     "[1] Frequency"=&RandomRangeFloat(Minimum=6, Maximum=10)
     "[1] Amplitude Min/Max"=<0 4>
    )
    Index 0 Param 2=1
    Index 0 Param 3=1
    Module.Index0Float0Write=true
    Module.Index0Float1Write=true
    Module.Index0Float2Write=true
    Module.Index0Float3Write=true
   )
   &SubUVAnimation(
    Random Start Frame=true
    "Sprite Renderer "=DataInterface() {propertyValues {SpriteRenderer=&self:Runes_0.NiagaraSpriteRendererProperties_0}}
    SubUV Lookup Index:NiagaraFloat=&Add_Float(
     A=&Multiply_Float(A=$Engine.System.Age, B=&RandomRangeFloat(Minimum=0.05, Maximum=0.2))
     B=&RandomRangeFloat(Minimum=0, Maximum=1)
    )
    SubUV Lookup Index Scale:NiagaraFloat=&FloatFromCurve(
     FloatCurve=DataInterface() {propertyValues {Curve {keys [<at 0 4>,<at 0.1 1>,<at 1 1>]}}}
     Scale Curve=&RandomRangeFloat(Minimum=1, Maximum=2)
    )
    UseEndFrame=false
    UseStartFrame=false
   ) {module=SubUVAnimation001}
   &ScaleColor(
    Scale Alpha=&FloatFromCurve(FloatCurve=DataInterface() {propertyValues {Curve {keys [{interpMode=RCIM_Cubic, tangentMode=RCTM_User, time=0, value=1}, <at 1 0>]}}})
    Scale RGB=&Multiply_VectorByFloat(
     Vector=<1 1 1>
     Float=&FloatFromCurve(
      FloatCurve=DataInterface() {propertyValues {Curve {keys [<at 0 1>, {interpMode=RCIM_Cubic, tangentMode=RCTM_User, time=0.2, value=3}, <at 0.4 1>, <at 1 1>]}}}
      Scale Curve=2
     )
    )
    ScaleA=true
    ScaleRGB=true
   )
   &Drag(
    Drag=&FloatFromCurve(
     FloatCurve=DataInterface() {propertyValues {Curve {
        defaultValue=1
        keys [
         {interpMode=RCIM_Cubic, tangentMode=RCTM_User, time=0.4406, value=0.3539}
         {arriveTangent=-0.8348, interpMode=RCIM_Cubic, leaveTangent=-0.8348, tangentMode=RCTM_User, time=0.498, value=1}
         {arriveTangent=-0.2088, interpMode=RCIM_Cubic, leaveTangent=-0.2088, tangentMode=RCTM_User, time=0.6463, value=0.7444}
         <at 1 0.6934 RCIM_Cubic>
        ]
       }}}
     CurveIndex=$Emitter.NormalizedLoopAge
     Scale Curve=&RandomRangeFloat(Minimum=15, Maximum=10)
    )
    Use Linear Drag=true
    Use Rotational Drag=false
   )
   &CurlNoiseForce(
    Bias Noise=false
    Noise Frequency=&RandomRangeFloat(Minimum=10, Maximum=50)
    Noise Strength=&Multiply_Float(
     A=&RandomRangeFloat(Minimum=100, Maximum=300)
     B=&FloatFromCurve(
      FloatCurve=DataInterface() {propertyValues {Curve {
         defaultValue=1
         keys [
          {interpMode=RCIM_Cubic, tangentMode=RCTM_User, time=0.4435, value=1}
          {interpMode=RCIM_Cubic, tangentMode=RCTM_User, time=0.4645, value=0.2434}
          {arriveTangent=2.6614, interpMode=RCIM_Cubic, leaveTangent=2.6614, tangentMode=RCTM_User, time=0.9853, value=0.6612}
         ]
        }}}
      CurveIndex=$Emitter.NormalizedLoopAge
     )
    )
    Pan Noise=false
   )
   &SolveForcesAndVelocity(Clamp Velocity=false, Limit Acceleration=false)
   /SetVariables(Particles.Position=&Lerp_Vector(
     A=Linked() {linkedVariable {name=StackContext.Position, type=NiagaraPosition}}
     B=Linked() {linkedVariable {name=Particles.ShapeLocation.ShapePosition, type=NiagaraPosition}}
     Alpha=0
    )) {isSetParameters=true}
  ]
  rendererClasses [&NiagaraSpriteRendererProperties]
  renderers [
   {rendererClass=&NiagaraSpriteRendererProperties}
   {rendererClass=&NiagaraSpriteRendererProperties, rendererIndex=1}
   {rendererClass=&NiagaraSpriteRendererProperties, rendererIndex=2}
  ]
 }
 StartGlow {
  bEnabled=false
  emitterUpdateScript [
   &EmitterState(
    Life Cycle Mode=::NewEnumerator1
    Loop Behavior=::NewEnumerator1
    Loop Delay=&Add_Float(A=$System.Delay, B=0.7)
    Loop Duration=3
    UseLoopDelay=true
   )
   &SpawnBurst_Instantaneous(Use Spawn Probability=false, UseLoopCountLimit=false)
  ]
  particleSpawnScript [&InitializeParticle(
    Color=&SelectLinearColorFromArray(
     Array Sampling Mode=::NewEnumerator1
     Direct Array Index=0
     Color Selection Array=DataInterface() {propertyValues {ColorData [<rgba 1 0.3784 0 1>, <rgba 0 0.4916 1 1>, <rgba 1 0.3784 0 1>]}}
     Random Seed=?
    )
    Lifetime=3
    Sprite Size Mode=::NewEnumerator3
    Uniform Sprite Size=60
    UsePositionOffset=false
    Write Material Random=true
   )]
  particleUpdateScript [
   &ParticleState()
   &SkeletalMeshLocation(Bone / Socket Index=11, Bone Sampling Mode=::NewEnumerator4, MaskSpawnGroup=false, Skeletal Mesh=$System.Cubes)
   &ScaleColor(
    Scale Alpha=&FloatFromCurve(
     FloatCurve=DataInterface() {propertyValues {Curve {
        defaultValue=1
        keys [
         {arriveTangent=6.8265, interpMode=RCIM_Cubic, leaveTangent=6.8265, tangentMode=RCTM_User, time=0, value=0}
         {interpMode=RCIM_Cubic, tangentMode=RCTM_User, time=0.5, value=1}
         <at 1 0 RCIM_Cubic>
        ]
       }}}
     Scale Curve=0.5
    )
    ScaleA=true
    ScaleRGB=false
   )
  ]
  rendererClasses [&NiagaraSpriteRendererProperties]
  renderers [{rendererClass=&NiagaraSpriteRendererProperties}]
 }
 Streaks {
  emitterSpawnScript [/SetVariables(
    Emitter.Linear Color:LinearColor=&SelectLinearColorFromArray(
     Array Sampling Mode=::NewEnumerator1
     Direct Array Index=&MakeIntFromFloat(Float=$User.Variant, Conversion Mode=::NewEnumerator1)
     Color Selection Array=DataInterface() {propertyValues {ColorData [<rgba 1 0.6762 0 1>, <rgba 0 0.5627 1 1>, <rgba 1 0.0119 0.007 1>]}}
     Random Seed=?
    )
    Emitter.Linear Color001:LinearColor=&SelectLinearColorFromArray(
     Array Sampling Mode=::NewEnumerator1
     Direct Array Index=&MakeIntFromFloat(Float=$User.Variant, Conversion Mode=::NewEnumerator1)
     Color Selection Array=DataInterface() {propertyValues {ColorData [<rgba 0.7078 0.202 0 1>, <rgba 0 0.2454 0.7078 1>, <rgba 0.7078 0.0104 0 1>]}}
     Random Seed=?
    )
   ) {isSetParameters=true}]
  emitterUpdateScript [
   &EmitterState(
    Inactive Response=::NewEnumerator1
    Life Cycle Mode=::NewEnumerator1
    Loop Behavior=::NewEnumerator1
    Loop Delay=&Add_Float(A=$System.Delay, B=1.5)
    Loop Duration=4.7
    UseLoopDelay=true
   )
   &SpawnRate(
    SpawnRate=&FloatFromCurve(
     FloatCurve=DataInterface() {propertyValues {Curve {
        defaultValue=1
        keys [
         {interpMode=RCIM_Cubic, tangentMode=RCTM_User, time=0.2, value=0.0544}
         {
          arriveTangent=18.2952
          arriveTangentWeight=3.147
          interpMode=RCIM_Cubic
          leaveTangentWeight=2.1513
          tangentMode=RCTM_Break
          time=0.6127
          value=2.9463
         }
         {arriveTangentWeight=0.7146, interpMode=RCIM_Cubic, leaveTangentWeight=0.2805, tangentMode=RCTM_User, time=0.8505, value=2.9906}
         <at 0.986 0 RCIM_Cubic>
        ]
       }}}
     CurveIndex=$Emitter.NormalizedLoopAge
     Scale Curve=30
    )
    Use Spawn Probability=false
   )
  ]
  particleSpawnScript [
   &PartitionParticles()
   &InitializeParticle(
    Color=<rgba 10 10 10 1>
    Lifetime Max=0.2
    Lifetime Min=0.3
    Lifetime Mode=::NewEnumerator1
    Sprite Size Mode=::NewEnumerator1
    Sprite UV Mode=::NewEnumerator1
    UsePositionOffset=false
    Write Material Random=true
   )
   &CameraOffset(Camera Offset Amount=30)
  ]
  particleUpdateScript [
   &ParticleState()
   &SkeletalMeshLocation(Bone Sampling Mode=::NewEnumerator4, MaskSpawnGroup=false, Skeletal Mesh=$System.Cubes)
   &ShapeLocation(
    MaskSpawnGroup=false
    Offset Mode=::NewEnumerator0
    Override Seed=false
    Rotation Mode=::NewEnumerator1
    Shape Origin=$Particles.SkeletalMeshLocation.SampledPosition
    Sphere Radius=&RandomRangeFloat(Minimum=20, Maximum=40)
    Sphere Surface Distribution=1
    UseOwnerScale=true
   )
   &ScaleSpriteSize(
    Scale Factor=&Multiply_Vector2D(A=&RandomRangeVector2D(Minimum=<2 2>, Maximum=<3 3>), B=<0.8 1>)
    Scale Sprite Size Mode=::NewEnumerator2
   )
   &ScaleSpriteSize(
    Non-Uniform Curve Scale=<1 2>
    Non-Uniform Curve Sprite Scale=DataInterface() {propertyValues {
      XCurve {keys [
        {arriveTangent=0.8609, interpMode=RCIM_Cubic, leaveTangent=0.8609, tangentMode=RCTM_User, time=0, value=0.6}
        {arriveTangent=-0.9108, interpMode=RCIM_Cubic, leaveTangent=-0.9108, tangentMode=RCTM_User, time=0.3, value=0.45}
        {arriveTangent=0.0224, interpMode=RCIM_Cubic, leaveTangent=0.0224, tangentMode=RCTM_User, time=1, value=0.5}
       ]}
      YCurve {keys [
        {arriveTangent=0.8801, interpMode=RCIM_Cubic, leaveTangent=0.8801, tangentMode=RCTM_User, time=0, value=0.3}
        {arriveTangent=0.6586, interpMode=RCIM_Cubic, leaveTangent=0.6586, tangentMode=RCTM_User, time=0.2, value=0.85}
        {arriveTangent=-0.4778, interpMode=RCIM_Cubic, leaveTangent=-0.4778, tangentMode=RCTM_User, time=1, value=2}
       ]}
     }}
    Scale Sprite Size Mode=::NewEnumerator3
   ) {module=ScaleSpriteSize001}
   &Drag(Use Linear Drag=true, Use Rotational Drag=true)
   &PointAttractionForce(
    Attraction Strength=0.1
    Attractor Position:NiagaraPosition=$Particles.SkeletalMeshLocation.SampledPosition
    Kill Within Radius=false
    Module.Use Falloff=true
    UseAttractorOffset=false
   )
   &SolveForcesAndVelocity(Clamp Velocity=false, Limit Acceleration=false)
   &ScaleColor(
    Scale Mode=::NewEnumerator1
    Scale RGBA=&MakeFloat4FromVectorAndFloat(
     "Vector (RGB)"=&Lerp_Vector(
      A=<1 0.4 0.4>
      B=<1 1 1>
      Alpha=&FloatFromCurve(FloatCurve=DataInterface() {propertyValues {Curve {keys [{arriveTangent=-4.0059, interpMode=RCIM_Cubic, leaveTangent=-4.0059, tangentMode=RCTM_User, time=0.5, value=1}, <at 1 0 RCIM_Cubic>]}}})
     )
     "Float (Alpha)"=&RandomRangeFloat(Minimum=0.2, Maximum=0.1)
    )
    ScaleRGBA=true
   ) {module=ScaleColor002}
   &DynamicMaterialParameters(
    Index 0 Param 1=&FloatFromCurve(FloatCurve=DataInterface() {propertyValues {Curve {keys [
         {arriveTangent=-0.1846, time=0, value=0}
         {arriveTangent=1.3385, interpMode=RCIM_Cubic, leaveTangent=1.1594, tangentMode=RCTM_Break, time=0.5, value=0}
         {arriveTangent=2, interpMode=RCIM_Cubic, leaveTangent=2, time=0.8, value=0.5}
         {arriveTangent=3.0091, interpMode=RCIM_Cubic, leaveTangent=3.0091, tangentMode=RCTM_User, time=1, value=1}
        ]}}})
    Index 0 Param 2=1
    Index 0 Param 4=&FloatFromCurve(FloatCurve=DataInterface() {propertyValues {Curve {keys [<at 0 0 RCIM_Cubic>,<at 1 1 RCIM_Cubic>]}}}, Scale Curve=0.8)
    Module.Index0Param0Write=true
    Module.Index0Param1Write=true
    Module.Index0Param2Write=true
    Module.Index0Param3Write=true
   )
  ]
  rendererClasses [&NiagaraSpriteRendererProperties]
  renderers [{rendererClass=&NiagaraSpriteRendererProperties}]
 }
 VariantColorTest {
  bEnabled=false
  emitterUpdateScript [
   &EmitterState()
   &SpawnBurst_Instantaneous(Use Spawn Probability=false, UseLoopCountLimit=false)
  ]
  particleSpawnScript [&InitializeParticle(
    Color=&SelectLinearColorFromArray(
     Array Sampling Mode=::NewEnumerator1
     Direct Array Index=0
     Color Selection Array=DataInterface() {propertyValues {ColorData [<rgba 1 0.3223 0 1>, <rgba 0 0.6788 1 1>, <rgba 0.8548 0 1 1>]}}
     Random Seed=?
    )
    Lifetime=$Emitter.CurrentLoopDuration
    UsePositionOffset=false
    Write Material Random=true
   )]
  particleUpdateScript [&ParticleState()]
  rendererClasses [&NiagaraSpriteRendererProperties]
  renderers [{rendererClass=&NiagaraSpriteRendererProperties}]
 }
 WarpEnergy {
  bEnabled=false
  emitterUpdateScript [
   &EmitterState(
    Life Cycle Mode=::NewEnumerator1
    Loop Behavior=::NewEnumerator1
    Loop Delay=&Add_Float(A=$System.Delay, B=1.5)
    UseLoopDelay=true
   )
   &SpawnBurst_Instantaneous(Use Spawn Probability=false, UseLoopCountLimit=false)
  ]
  particleSpawnScript [&InitializeParticle(
    Color=&MergeHelper_ColorSwitch(0=<rgba 3 0.965 0 1>, 1=<rgba 0 1.0468 2 1>, 2=<rgba 3 0.965 0 1>, 3=?, Index=0, Num=::NewEnumerator1)
    Lifetime=5
    Mesh Scale=&Multiply_VectorByFloat(Vector=<-2 -4 -3.5>, Float=0.3)
    Mesh Scale Mode=::NewEnumerator4
    UsePositionOffset=false
    Write Material Random=true
   )]
  particleUpdateScript [
   &ParticleState()
   &ScaleMeshSize(Scale Factor=&VectorFromFloat(Value=&FloatFromCurve(FloatCurve=DataInterface() {propertyValues {Curve {
         defaultValue=1
         keys [{interpMode=RCIM_Cubic, tangentMode=RCTM_User, time=0.9682, value=1}, {interpMode=RCIM_Cubic, tangentMode=RCTM_User, time=1, value=0}]
        }}})))
   &UpdateMeshOrientation(
    Facing Interpolation=false
    Facing Position:NiagaraPosition=&ReturnCameraPosition(Coordinate Space:ENiagaraCoordinateSpace=::Simulation, Camera Query:NiagaraDataInterfaceCamera=DataInterface())
    Orientation Method=::NewEnumerator2
    Override Quat=false
    TransformFacingOrigin=false
    TransformFacingPosition=false
    TransformUp=false
   ) {enabled=false}
   /SetVariables(
    Particles.DynPar_ChromaWave:Vector4f=&MakeVector4(
     X=&RandomRangeFloat(Minimum=0, Maximum=1)
     Y=&FloatFromCurve(
      FloatCurve=DataInterface() {propertyValues {Curve {
         defaultValue=1
         keys [
          {arriveTangent=0.2806, interpMode=RCIM_Cubic, leaveTangent=0.2806, tangentMode=RCTM_User, time=0, value=0}
          {arriveTangent=2.2839, interpMode=RCIM_Cubic, leaveTangent=2.2839, tangentMode=RCTM_User, time=1, value=1}
         ]
        }}}
      Scale Curve=2
     )
     Z=0
     W=0
    )
    Particles.DynPar_SoftWave:Vector4f=&MakeVector4(X=3, Y=1.5, Z=0, W=0)
   ) {isSetParameters=true, module=SetVariables_#8}
   &SkeletalMeshLocation(Bone Sampling Mode=::NewEnumerator4, MaskSpawnGroup=false, Skeletal Mesh=$System.Cubes)
   /SetVariables(Particles.MeshOrientation=&OrientMeshToVector(
     Look At Direction=&ReturnCameraVector(Camera Vectors=::NewEnumerator1, Coordinate Space=::Simulation)
     Reference Axis Vector=<1 0 0>
    )) {isSetParameters=true, module=SetVariables_#9}
   &ScaleColor(
    Scale Alpha=&FloatFromCurve(FloatCurve=DataInterface() {propertyValues {Curve {
        defaultValue=1
        keys [
         {arriveTangent=0.1462, interpMode=RCIM_Cubic, leaveTangent=0.1462, tangentMode=RCTM_User, time=0, value=0}
         {interpMode=RCIM_Cubic, tangentMode=RCTM_User, time=0.7269, value=1}
        ]
       }}})
    ScaleA=true
    ScaleRGB=false
   )
  ]
  rendererClasses [&NiagaraMeshRendererProperties]
  renderers [{rendererClass=&NiagaraMeshRendererProperties, rendererIndex=0}]
 }
 Wave02 {
  emitterUpdateScript [
   &EmitterState(
    Inactive Response=::NewEnumerator1
    Life Cycle Mode=::NewEnumerator1
    Loop Behavior=::NewEnumerator1
    Loop Delay=&Add_Float(A=$System.Delay, B=0.17)
    Loop Duration=12
    UseLoopDelay=true
   )
   &SpawnBurst_Instantaneous(Spawn Time=9.29, Use Spawn Probability=false, UseLoopCountLimit=false) {module=SpawnBurst_Instantaneous001}
   &SpawnBurst_Instantaneous(Spawn Count=1, Use Spawn Probability=false)
  ]
  particleSpawnScript [
   &InitializeParticle(
    Color=&SelectLinearColorFromArray(
     Array Sampling Mode=::NewEnumerator1
     Direct Array Index=&MakeIntFromFloat(Float=$User.Variant, Conversion Mode=::NewEnumerator1)
     Color Selection Array=DataInterface() {propertyValues {ColorData [<rgba 2 0.5985 0 1>, <rgba 0 1.0551 3 1>, <rgba 2 0.0013 0 1>]}}
     Random Seed=?
    )
    Lifetime=0.3
    Mesh Scale Mode=::NewEnumerator3
    Mesh Uniform Scale=4
    Sprite Rotation Mode=::NewEnumerator2
    UsePositionOffset=false
    Write Material Random=true
   )
   &InitialMeshOrientation(Orientation Vector=<0 0 1>, Use Orientation Vector=true, Use Rotation Vector=false)
   &CameraOffset()
  ]
  particleUpdateScript [
   &ParticleState()
   &DynamicMaterialParameters(
    Index 0 Param 1=&FloatFromCurve(FloatCurve=DataInterface() {propertyValues {Curve {keys [
         {arriveTangent=-2.1015, interpMode=RCIM_Cubic, leaveTangent=-2.1015, tangentMode=RCTM_User, time=0, value=0.8532}
         {arriveTangent=-1.0032, interpMode=RCIM_Cubic, leaveTangent=-1.0032, time=0.1138, value=0.1165}
         <at 1 -0.15 RCIM_Cubic>
        ]}}})
    Index 0 Param 2=&FloatFromCurve(FloatCurve=DataInterface() {propertyValues {Curve {keys [{arriveTangent=-0.3464, interpMode=RCIM_Cubic, leaveTangent=-0.3464, tangentMode=RCTM_User, time=0, value=1}, <at 1 0.9>]}}})
    Index 0 Param 3=&FloatFromCurve(FloatCurve=DataInterface() {propertyValues {Curve {keys [<at 0 0>,<at 0.15 1>,<at 1 1.2>]}}})
    Module.Index0Float0Write=true
    Module.Index0Float1Write=true
    Module.Index0Float2Write=true
    Module.Index0Float3Write=true
   )
   &ScaleColor(
    Scale Alpha=&FloatFromCurve(
     FloatCurve=DataInterface() {propertyValues {Curve {keys [
         <at 0.1 0.8131 RCIM_Cubic>
         {interpMode=RCIM_Cubic, leaveTangent=-2.4889, tangentMode=RCTM_Break, time=0.3047, value=0.8}
         {arriveTangent=-1.1506, interpMode=RCIM_Cubic, leaveTangent=-1.1506, time=0.532, value=0.181}
         <at 1 0>
        ]}}}
     Scale Curve=&FN_FrontEndFloatModifier(Gameplay=1, Front End=0.6)
    )
    Scale RGB=&VectorFromFloat(Value=&FloatFromCurve(FloatCurve=DataInterface() {propertyValues {Curve {keys [<at 0 10>,<at 0.15 3>,<at 1 1>]}}}))
    ScaleA=true
    ScaleRGB=true
   ) {module=ScaleColor001}
   &ScaleMeshSize(Scale Factor=&VectorFromFloat(Value=&FloatFromCurve(
      FloatCurve=DataInterface() {propertyValues {Curve {keys [<at 0 0.5836>,<at 0.0592 1.6946>,<at 0.3467 2>]}}}
      Scale Curve=&Multiply_Float(
       A=0.115
       B=&Multiply_Float(
        A=0.7084
        B=&FN_LOD_Distance_Remap(Start Distance=350, End Distance=380, Minimum Output=0.7, Maximum Output=1, Clamp Output=true)
       )
      )
     )))
   &SkeletalMeshLocation(Bone / Socket Index=11, Bone Sampling Mode=::NewEnumerator4, MaskSpawnGroup=false, Skeletal Mesh=$System.Cubes)
   /SetVariables(Particles.MeshOrientation=&OrientMeshToVector(
     Look At Direction=&ReturnCameraVector(Camera Vectors=::NewEnumerator1, Coordinate Space=::Simulation)
     Reference Axis Vector=<1 0 0>
    )) {isSetParameters=true}
  ]
  rendererClasses [&NiagaraMeshRendererProperties]
  renderers [{rendererClass=&NiagaraMeshRendererProperties, rendererIndex=0}]
 }
}
summary {
 systemName=NS_Emote_SpacePlunge_Burst
 emitters [
  {
   bEnabled=false
   emitterName=Glow
   rendererClasses [&NiagaraSpriteRendererProperties, &NiagaraLightRendererProperties]
  }
  {emitterName=Cubes, rendererClasses [&NiagaraMeshRendererProperties]}
  {emitterName=Runes, rendererClasses [&NiagaraSpriteRendererProperties]}
  {emitterName=Lightning, rendererClasses [&NiagaraMeshRendererProperties]}
  {emitterName=Flare, rendererClasses [&NiagaraSpriteRendererProperties]}
  {bEnabled=false, emitterName=WarpEnergy, rendererClasses [&NiagaraMeshRendererProperties]}
  {bEnabled=false, emitterName=Energy, rendererClasses [&NiagaraMeshRendererProperties]}
  {bEnabled=false, emitterName=VariantColorTest, rendererClasses [&NiagaraSpriteRendererProperties]}
  {emitterName=Glitch, rendererClasses [&NiagaraSpriteRendererProperties]}
  {bEnabled=false, emitterName=Hunt_Lightning, rendererClasses [&NiagaraMeshRendererProperties]}
  {emitterName=Streaks, rendererClasses [&NiagaraSpriteRendererProperties]}
  {
   emitterName=Dark_Glow
   rendererClasses [
    &NiagaraSpriteRendererProperties
    &NiagaraLightRendererProperties
    &NiagaraMeshRendererProperties
   ]
  }
  {emitterName=Wave02, rendererClasses [&NiagaraMeshRendererProperties]}
  {bEnabled=false, emitterName=StartGlow, rendererClasses [&NiagaraSpriteRendererProperties]}
  {bEnabled=false, emitterName=Dark_Glow001, rendererClasses [&NiagaraSpriteRendererProperties]}
 ]
 userVariables [
  User.bIsFrontEnd {type=NiagaraBool}
  User.bIsFrontEndPreview {type=NiagaraBool}
  User.Variant {type=NiagaraFloat, defaultValue=0, description=""}
 ]
}
