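# Niagara system for the hero effect of the SpacePlunge emote, showing mesh lightning bursts and a flash wave.
# User.BurstIndex sets the lightning spawn count and lerps wave size, while User.Variant picks warm yellow/orange, blue or red tints.
# Curves drive fade, scale and dynamic material parameters, with front-end dimming, time-of-day scaling on lightning and distance-based wave sizing.
# Only Hunt_Lightning and Wave02 are enabled; both impact flares and WarpEnergy are disabled.
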NiagaraSystem /BRCosmetics/Animation/Game/MainPlayer/Emotes/SpacePlunge/FX/NS_Emote_SpacePlunge_Hero
dependencies {
 usedDataInterfaces [
  &NiagaraDataInterfaceArrayColor
  &NiagaraDataInterfaceCurve
  &NiagaraDataInterfacePlatformSet
  &NiagaraDataInterfaceColorCurve
 ]
 usedDynamicInputs []
 usedModules [&SystemState]
}
emitters {
 Hunt_Lightning {
  emitterUpdateScript [
   &EmitterState(Life Cycle Mode=::NewEnumerator1, Loop Behavior=::NewEnumerator1, Loop Duration=0.3, UseLoopDelay=true)
   &SpawnRate(SpawnRate=10, Use Spawn Probability=false) {enabled=false}
   &SpawnBurst_Instantaneous(Spawn Count=$User.BurstIndex, Use Spawn Probability=false, UseLoopCountLimit=false)
  ]
  particleSpawnScript [
   &InitializeParticle(
    Color=&MergeHelper_ColorSwitch(
     0=<rgba 5 3.6617 0 1>
     1=<rgba 0 3.1555 5 1>
     2=<rgba 5 0.1695 0.127 1>
     3=?
     Index=&MakeIntFromFloat(Float=$User.Variant, Conversion Mode=::NewEnumerator1)
     Num=::NewEnumerator1
    )
    Lifetime Max=0.5
    Lifetime Min=0.25
    Lifetime Mode=::NewEnumerator1
    Mesh Scale=&Multiply_VectorByFloat(
     Vector=&Lerp_Vector(
      A=&RandomRangeVector(Minimum=<0.5 0.4 0.4>, Maximum=<0.75 0.6 0.6>)
      B=&RandomRangeVector(Minimum=<1 1 1>, Maximum=<1.25 1.5 1.5>)
      Alpha=&FloatFromCurve(
       FloatCurve=DataInterface() {propertyValues {Curve {keys [{interpMode=RCIM_Cubic, tangentMode=RCTM_User, time=0.5, value=0}, <at 1 1>]}}}
       CurveIndex=&RandomRangeFloat(Minimum=0, Maximum=1)
      )
     )
     Float=0.4
    )
    Mesh Scale Mode=::NewEnumerator4
    UsePositionOffset=true
    Write Material Random=true
   )
   &ShapeLocation(Box Size=<0.5 0.5 0.5>, MaskSpawnGroup=false, Override Seed=false, Shape Primitive=::NewEnumerator2, Surface Only=false, UseOwnerScale=true)
   &InitialMeshOrientation(Rotation=&RandomRangeVector(Minimum=<-0.5 -0.5 -0.5>, Maximum=<0.5 0.5 0.5>), Use Rotation Vector=true) {enabled=false, module=InitialMeshOrientation002}
   /SetVariables(Particles.Random=&RandomRangeFloat(Minimum=0, Maximum=1)) {isSetParameters=true}
   &AddVelocity(Velocity=<1 -1 0>)
  ]
  particleUpdateScript [
   &ParticleState()
   &ScaleMeshSize()
   &ScaleColor(
    Scale Alpha=&Multiply_Float(
     A=10
     B=&FloatFromCurve(FloatCurve=DataInterface() {propertyValues {Curve {keys [
          {interpMode=RCIM_Cubic, tangentMode=RCTM_User, time=0, value=0}
          {interpMode=RCIM_Cubic, tangentMode=RCTM_User, time=0.1, value=0}
          {interpMode=RCIM_Cubic, tangentMode=RCTM_User, time=0.2, value=0.8}
          {interpMode=RCIM_Cubic, tangentMode=RCTM_User, time=1, value=0}
         ]}}})
    )
    Scale RGB=&Multiply_VectorByFloat(
     Vector=&Multiply_VectorByFloat(Vector=<1 1 1>, Float=&FN_FrontEndFloatModifier(Gameplay=1, Front End=0.5))
     Float=&Multiply_Float(
      A=&RandomRangeFloat(Minimum=0.5, Maximum=1, Evaluation Type=::NewEnumerator1)
      B=&Multiply_Float(A=&FNTimeOfDayScalar(Night Scale=0.2, Morning Scale=1, Evening Scale=1), B=10)
     )
    )
    ScaleA=true
    ScaleRGB=true
   )
   &DynamicMaterialParameters(
    Index 0 Param 1=&FloatFromCurve(FloatCurve=DataInterface() {propertyValues {Curve {keys [{tangentMode=RCTM_User, time=0, value=1}, {tangentMode=RCTM_User, time=1, value=1}]}}})
    Index 0 Param 2=0.7
    Index 0 Param 3=&FloatFromCurve(FloatCurve=DataInterface() {propertyValues {Curve {keys [
         {arriveTangent=4.0072, interpMode=RCIM_Cubic, leaveTangent=4.0072, tangentMode=RCTM_User, time=0, value=0}
         {arriveTangent=2.7083, interpMode=RCIM_Cubic, leaveTangent=2.7083, time=0.2543, value=0.6225}
         {arriveTangent=1.8473, interpMode=RCIM_Cubic, leaveTangent=1.8473, time=0.6, value=1.625}
         {arriveTangent=1.6993, interpMode=RCIM_Cubic, leaveTangent=1.6993, tangentMode=RCTM_User, time=1, value=2}
        ]}}})
    Index 0 Param 4=5
    Index 1 Param 1=0.7
    Index 1 Param 2=$Particles.Random
    Index 1 Param 3=1.3
    Index 1 Param 4=1
    Index 2 Param 1=&FloatFromCurve(FloatCurve=DataInterface() {propertyValues {Curve {keys [<at 0 1>,<at 1 0.3>]}}})
    Module.Index0Float0Write=true
    Module.Index0Float1Write=true
    Module.Index0Float2Write=true
    Module.Index0Float3Write=true
    Module.Index1Float0Write=true
    Module.Index1Float1Write=true
    Module.Index1Float2Write=true
    Module.Index1Float3Write=true
    Module.Index2Float0Write=true
    Module.Index2Float1Write=true
    Module.Index2Float2Write=true
    Module.Index2Float3Write=true
    Write Parameter Index 1=::NewEnumerator1
    Write Parameter Index 2=::NewEnumerator1
   )
   &SolveForcesAndVelocity(Clamp Velocity=false, Limit Acceleration=false)
   /SetVariables(Particles.Lightning glow color:LinearColor=&MakeLinearColorFromVectorAndFloat(
     "Vector (RGB)"=&MakeVectorFromLinearColorRGB(Color=&MultiplyLinearColor(
       Linear Color=&SelectLinearColorFromArray(
        Array Sampling Mode=::NewEnumerator1
        Direct Array Index=0
        Color Selection Array=DataInterface() {propertyValues {ColorData [<rgba 4 1.6071 0.1968 1>, <rgba 0.3366 2.0701 4 1>, <rgba 4 1.6071 0.1968 1>]}}
        Random Seed=?
       )
       Scale Factor=&MakeVector4_FromFloat(Float=&RandomRangeFloat(Minimum=0.5, Maximum=1, Evaluation Type=::NewEnumerator1))
      ))
     "Float (Alpha)"=&Multiply_Float(
      A=&FloatFromCurve(FloatCurve=DataInterface() {propertyValues {Curve {keys [<at 0 1 RCIM_Cubic>, {interpMode=RCIM_Cubic, leaveTangent=-3.6307, time=0.9, value=0}]}}})
      B=1
     )
    )) {isSetParameters=true}
   /SetVariables(Particles.Glow Dy02:Vector4f=&MakeVector4(
     X=&FloatFromCurve(FloatCurve=DataInterface() {propertyValues {Curve {keys [{tangentMode=RCTM_User, time=0, value=0.2}, {tangentMode=RCTM_User, time=0.5, value=1}]}}})
     Y=0.7
     Z=0
     W=5
    )) {isSetParameters=true}
   /SetVariables(Particles.Glow Dy03:Vector4f=&MakeVector4(
     X=&FloatFromCurve(
      FloatCurve=DataInterface() {propertyValues {Curve {keys [<at 0 1>,<at 1 0.5>]}}}
      Scale Curve=&MaskFloat(Mask Source=1, Mask Parameter=0, Mask=2, Masked Float=0.7, Passthrough Float=0.5)
     )
     Y=$Particles.Random
     Z=1.3
     W=1
    )) {isSetParameters=true}
  ]
  rendererClasses [&NiagaraMeshRendererProperties]
  renderers [
   {rendererClass=&NiagaraMeshRendererProperties, rendererIndex=0}
   {rendererClass=&NiagaraMeshRendererProperties, rendererIndex=1}
  ]
 }
 Impact_Flare {
  bEnabled=false
  emitterUpdateScript [
   &EmitterState(Life Cycle Mode=::NewEnumerator1, Loop Behavior=::NewEnumerator1, Loop Duration=0.5, UseLoopDelay=false)
   &SpawnBurst_Instantaneous(Spawn Count=1, Use Spawn Probability=false)
  ]
  particleSpawnScript [
   &InitializeParticle(
    Color=&SelectLinearColorFromArray(
     Array Sampling Mode=::NewEnumerator1
     Direct Array Index=0
     Color Selection Array=DataInterface() {propertyValues {ColorData [<rgba 1 0.3784 0 1>, <rgba 0 0.4916 1 1>, <rgba 1 0.3784 0 1>]}}
     Random Seed=?
    )
    Lifetime=0.25
    Mesh Scale Mode=::NewEnumerator3
    Mesh Uniform Scale=0.5
    Sprite Size Mode=::NewEnumerator3
    Uniform Sprite Size=30
    UsePositionOffset=false
    Write Material Random=true
   ) {module=InitializeParticle001}
   &CameraOffset(Camera Offset Amount=20)
   &InitialMeshOrientation(
    Rotation=&RandomRangeVector(Minimum=<-1 -1 -1>, Maximum=<1 1 1>)
    Use Orientation Vector=true
    Use Rotation Vector=true
   ) {module=InitialMeshOrientation001}
  ]
  particleUpdateScript [
   &ParticleState()
   &ScaleColor(
    Scale Alpha=&FloatFromCurve(FloatCurve=DataInterface() {propertyValues {Curve {keys [{arriveTangent=-2.6501, interpMode=RCIM_Cubic, leaveTangent=-2.6501, tangentMode=RCTM_User, time=0, value=1}, <at 1 0>]}}})
    Scale RGB=&Multiply_VectorByFloat(
     Vector=&Multiply_VectorByFloat(
      Vector=<1 1 1>
      Float=&FN_PickFloatDependOnPlatform(TargetPlatformActive=1, TargetPlatformNoActive=0.2)
     )
     Float=&FN_FrontEndFloatModifier(Gameplay=&FNTimeOfDayScalar(Night Scale=0.5, Morning Scale=1, Evening Scale=1), Front End=0.4)
    )
    ScaleA=true
    ScaleRGB=true
   )
   &ScaleMeshSize(Scale Factor=&VectorFromFloat(Value=&FloatFromCurve(FloatCurve=DataInterface() {propertyValues {Curve {keys [{arriveTangent=1.5592, interpMode=RCIM_Cubic, leaveTangent=1.5592, tangentMode=RCTM_User, time=0, value=0.35}, <at 0.6 1.25>]}}})))
   &DynamicMaterialParameters(
    Index 0 Param 1=&FloatFromCurve(FloatCurve=DataInterface() {propertyValues {Curve {keys [<at 0 1 RCIM_Cubic>,<at 0.8 0 RCIM_Cubic>]}}})
    Index 0 Param 2=2
    Index 0 Param 3=1
    Index 0 Param 4=1
    Index 1 Param 1=&FloatFromCurve(FloatCurve=DataInterface() {propertyValues {Curve {keys [
         {arriveTangent=-2.3163, interpMode=RCIM_Cubic, time=0, value=2}
         {arriveTangent=-1.9019, interpMode=RCIM_Cubic, leaveTangent=-1.9019, tangentMode=RCTM_User, time=1, value=0.4}
        ]}}})
    Index 1 Param 2=0.4
    Index 1 Param 3=0.6
    Index 1 Param 4=1
    Index 2 Param 1=&FloatFromCurve(FloatCurve=DataInterface() {propertyValues {Curve {keys [<at 0.46 0>,<at 1 1>]}}})
    Index 2 Param 2=45
    Module.Index0Float0Write=true
    Module.Index0Float1Write=true
    Module.Index0Float2Write=true
    Module.Index0Float3Write=true
    Module.Index1Float0Write=true
    Module.Index1Float1Write=true
    Module.Index1Float2Write=true
    Module.Index1Float3Write=true
    Module.Index2Float0Write=true
    Module.Index2Float1Write=true
    Module.Index2Float2Write=true
    Module.Index2Float3Write=true
    Write Parameter Index 1=::NewEnumerator1
    Write Parameter Index 2=::NewEnumerator1
   ) {module=DynamicMaterialParameters001}
  ]
  rendererClasses [&NiagaraMeshRendererProperties]
  renderers [{rendererClass=&NiagaraMeshRendererProperties, rendererIndex=0}]
 }
 Impact_Flare001 {
  bEnabled=false
  emitterUpdateScript [
   &EmitterState(Life Cycle Mode=::NewEnumerator1, Loop Behavior=::NewEnumerator1, Loop Duration=0.5, UseLoopDelay=false)
   &SpawnBurst_Instantaneous(Spawn Count=1, Use Spawn Probability=false)
  ]
  particleSpawnScript [
   &InitializeParticle(
    Color Minimum=<rgba 1 1 1 1>
    Color Mode=::NewEnumerator2
    Lifetime=0.4
    Sprite Rotation Mode=::NewEnumerator2
    Sprite Size=<200 175>
    Sprite Size Mode=::NewEnumerator4
    UsePositionOffset=false
    Write Material Random=true
   ) {module=InitializeParticle001}
   &CameraOffset(Camera Offset Amount=20)
  ]
  particleUpdateScript [
   &ParticleState()
   &ScaleSpriteSize(
    Uniform Curve Scale=&Multiply_Float(A=&RandomRangeFloat(Minimum=0.8, Maximum=1), B=0.4)
    Uniform Curve Sprite Scale=DataInterface() {propertyValues {Curve {keys [{tangentMode=RCTM_User, time=0, value=0.8}, {tangentMode=RCTM_User, time=1, value=1}]}}}
   )
   &ScaleColor(
    Scale Alpha=&FloatFromCurve(
     FloatCurve=DataInterface() {propertyValues {Curve {keys [{interpMode=RCIM_Cubic, tangentMode=RCTM_User, time=0, value=1}, {interpMode=RCIM_Cubic, tangentMode=RCTM_User, time=1, value=0}]}}}
     Scale Curve=3
    )
    Scale RGB=&MakeVectorFromLinearColorRGB(Color=&MergeHelper_ColorSwitch(
      0=&ColorFromCurve(ColorCurve=DataInterface() {propertyValues {
         AlphaCurve {keys []}
         BlueCurve {keys [<at 0.0027 0>,<at 0.39 0 RCIM_Cubic>]}
         GreenCurve {keys [<at 0.0027 1.1569>,<at 0.39 0.1266 RCIM_Cubic>]}
         RedCurve {keys [<at 0.0027 1.5>,<at 0.39 0.4022 RCIM_Cubic>]}
        }})
      1=&ColorFromCurve(ColorCurve=DataInterface() {propertyValues {
         AlphaCurve {keys []}
         BlueCurve {keys [<at 0.0027 1.5>,<at 0.39 0.1771 RCIM_Cubic>]}
         GreenCurve {keys [<at 0.0027 1.0772>,<at 0.39 0.0935 RCIM_Cubic>]}
         RedCurve {keys [<at 0.0027 0>,<at 0.39 0 RCIM_Cubic>]}
        }})
      2=&ColorFromCurve(ColorCurve=DataInterface() {propertyValues {
         AlphaCurve {keys []}
         BlueCurve {keys [<at 0.0027 0>,<at 0.39 0 RCIM_Cubic>]}
         GreenCurve {keys [<at 0.0027 1.1569>,<at 0.39 0.1266 RCIM_Cubic>]}
         RedCurve {keys [<at 0.0027 1.5>,<at 0.39 0.4022 RCIM_Cubic>]}
        }})
      3=?
      Index=0
      Num=::NewEnumerator1
     ))
    ScaleA=true
    ScaleRGB=true
   )
   &DynamicMaterialParameters(
    Index 0 Param 1=&FloatFromCurve(FloatCurve=DataInterface() {propertyValues {Curve {keys [{tangentMode=RCTM_User, time=0, value=0}, {tangentMode=RCTM_User, time=1, value=1}]}}})
    Index 0 Param 2=&FloatFromCurve(FloatCurve=DataInterface() {propertyValues {Curve {keys [
         {arriveTangent=2.4836, interpMode=RCIM_Cubic, leaveTangent=2.4836, tangentMode=RCTM_User, time=0, value=0}
         {interpMode=RCIM_Cubic, tangentMode=RCTM_User, time=1, value=1}
        ]}}})
    Index 1 Param 1=3
    Index 1 Param 2=10
    Index 1 Param 3=1
    Index 1 Param 4=1
    Module.Index0Float0Write=true
    Module.Index0Float1Write=true
    Module.Index0Float2Write=true
    Module.Index0Float3Write=true
    Module.Index1Float0Write=true
    Module.Index1Float1Write=true
    Module.Index1Float2Write=true
    Module.Index1Float3Write=true
    Write Parameter Index 1=::NewEnumerator1
   )
   &FN_Sprite_Attributes(
    Color=&MakeLinearColorFromVectorAndFloat(
     "Vector (RGB)"=&Multiply_Vector(
      A=&MakeVectorFromLinearColorRGB(Color=<rgba 0.5 0.5 0.5 1>)
      B=&MakeVectorFromLinearColorRGB(Color=&ColorFromCurve(ColorCurve=DataInterface() {propertyValues {
          AlphaCurve {keys []}
          BlueCurve {keys [<at 1 0.0805 RCIM_Cubic>]}
          GreenCurve {keys [<at 1 0.099 RCIM_Cubic>]}
          RedCurve {keys [<at 1 0 RCIM_Cubic>]}
         }}))
     )
     "Float (Alpha)"=&Multiply_Float(
      A=&FloatFromCurve(
       FloatCurve=DataInterface() {propertyValues {Curve {keys [
           {interpMode=RCIM_Cubic, tangentMode=RCTM_User, time=0, value=1}
           {arriveTangent=-1.7926, interpMode=RCIM_Cubic, leaveTangent=-1.7926, tangentMode=RCTM_User, time=1, value=0}
          ]}}}
       Scale Curve=&FN_PickFloatDependOnPlatform(TargetPlatformActive=1, TargetPlatformNoActive=0.5)
      )
      B=1.3
     )
    )
    Create Color Attribute=true
    Create SpriteSize Attribute=true
    SpriteSize=&Multiply_Vector2D(
     A=<180 180>
     B=&Vector2DFromFloat(Value=&FloatFromCurve(
       FloatCurve=DataInterface() {propertyValues {Curve {keys [<at 0 1>,<at 1 3>]}}}
       Scale Curve=&Multiply_Float(A=$System.SizeScale, B=1)
      ))
    )
    Write to Module Namespace=true
   )
  ]
  rendererClasses [&NiagaraSpriteRendererProperties]
  renderers [
   {rendererClass=&NiagaraSpriteRendererProperties}
   {rendererClass=&NiagaraSpriteRendererProperties, rendererIndex=1}
  ]
 }
 WarpEnergy {
  bEnabled=false
  emitterUpdateScript [
   &EmitterState(Life Cycle Mode=::NewEnumerator1, Loop Behavior=::NewEnumerator1, Loop Duration=2, UseLoopDelay=true)
   &SpawnBurst_Instantaneous(Use Spawn Probability=false, UseLoopCountLimit=false)
  ]
  particleSpawnScript [
   &InitializeParticle(
    Color=&MergeHelper_ColorSwitch(0=<rgba 5 1.6083 0 1>, 1=<rgba 0 2.0935 4 1>, 2=<rgba 5 1.6083 0 1>, 3=?, Index=0, Num=::NewEnumerator1)
    Lifetime=0.4
    Mesh Scale=&Multiply_VectorByFloat(Vector=<2 4 3.5>, Float=0.15)
    Mesh Scale Mode=::NewEnumerator4
    UsePositionOffset=false
    Write Material Random=true
   )
   &InitialMeshOrientation(Orientation Vector=<0 1 0>, Use Orientation Vector=true, Use Rotation Vector=false)
  ]
  particleUpdateScript [
   &ParticleState()
   &ScaleMeshSize(Scale Factor=&VectorFromFloat(Value=&FloatFromCurve(FloatCurve=DataInterface() {propertyValues {Curve {
         defaultValue=1
         keys [
          {interpMode=RCIM_Cubic, leaveTangentWeight=0.0106, tangentMode=RCTM_User, time=0, value=0}
          {arriveTangentWeight=0.0106, interpMode=RCIM_Cubic, tangentMode=RCTM_User, time=0.0318, value=1}
         ]
        }}})))
   /SetVariables(
    Particles.DynPar_ChromaWave:Vector4f=&MakeVector4(
     X=&RandomRangeFloat(Minimum=0, Maximum=1)
     Y=&FloatFromCurve(
      FloatCurve=DataInterface() {propertyValues {Curve {
         defaultValue=0
         keys [
          {arriveTangent=-1.7937, interpMode=RCIM_Cubic, leaveTangent=-1.7937, tangentMode=RCTM_User, time=0, value=1}
          {arriveTangent=-0.2965, interpMode=RCIM_Cubic, leaveTangent=-0.2965, tangentMode=RCTM_User, time=1, value=0}
         ]
        }}}
      Scale Curve=0.5
     )
     Z=0
     W=0
    )
    Particles.DynPar_SoftWave:Vector4f=&MakeVector4(X=3, Y=1.5, Z=0, W=0)
   ) {isSetParameters=true}
   &ScaleColor(
    Scale Alpha=&FloatFromCurve(FloatCurve=DataInterface() {propertyValues {Curve {
        defaultValue=1
        keys [
         {arriveTangent=-2.047, interpMode=RCIM_Cubic, leaveTangent=-2.047, tangentMode=RCTM_User, time=0, value=1}
         {interpMode=RCIM_Cubic, tangentMode=RCTM_User, time=1, value=0}
        ]
       }}})
    ScaleA=true
    ScaleRGB=false
   )
  ]
  rendererClasses [&NiagaraMeshRendererProperties]
  renderers [{rendererClass=&NiagaraMeshRendererProperties, rendererIndex=0}]
 }
 Wave02 {
  emitterUpdateScript [
   &EmitterState(Inactive Response=::NewEnumerator1, Life Cycle Mode=::NewEnumerator1, Loop Behavior=::NewEnumerator1, Loop Duration=1, UseLoopDelay=true)
   &SpawnBurst_Instantaneous(Spawn Count=1, Use Spawn Probability=false)
  ]
  particleSpawnScript [
   &InitializeParticle(
    Color=&SelectLinearColorFromArray(
     Array Sampling Mode=::NewEnumerator1
     Direct Array Index=&MakeIntFromFloat(Float=$User.Variant, Conversion Mode=::NewEnumerator1)
     Color Selection Array=DataInterface() {propertyValues {ColorData [<rgba 1 0.469 0 1>, <rgba 0 0.4916 1 1>, <rgba 2 0.0431 0.0159 1>]}}
     Random Seed=?
    )
    Lifetime=0.3
    Mesh Scale Mode=::NewEnumerator3
    Mesh Uniform Scale=&Lerp_Float(A=1.5, B=3, Alpha=&MakeFloatFromInt(Int=$User.BurstIndex))
    Sprite Rotation Mode=::NewEnumerator2
    UsePositionOffset=false
    Write Material Random=true
   )
   &InitialMeshOrientation(Orientation Axis=<0 1 0>, Orientation Vector=<0 0 1>, Use Orientation Vector=true, Use Rotation Vector=false)
   &CameraOffset(Camera Offset Amount=20)
  ]
  particleUpdateScript [
   &ParticleState()
   &DynamicMaterialParameters(
    Index 0 Param 1=&FloatFromCurve(FloatCurve=DataInterface() {propertyValues {Curve {keys [
         {arriveTangent=-2.1015, interpMode=RCIM_Cubic, leaveTangent=-2.1015, tangentMode=RCTM_User, time=0, value=0.8532}
         {arriveTangent=-1.0032, interpMode=RCIM_Cubic, leaveTangent=-1.0032, time=0.1138, value=0.1165}
         <at 1 -0.15 RCIM_Cubic>
        ]}}})
    Index 0 Param 2=&FloatFromCurve(FloatCurve=DataInterface() {propertyValues {Curve {keys [{arriveTangent=-0.3464, interpMode=RCIM_Cubic, leaveTangent=-0.3464, tangentMode=RCTM_User, time=0, value=1}, <at 1 0.9>]}}})
    Index 0 Param 3=&FloatFromCurve(FloatCurve=DataInterface() {propertyValues {Curve {keys [<at 0 0>,<at 0.15 1>,<at 1 1.2>]}}})
    Module.Index0Float0Write=true
    Module.Index0Float1Write=true
    Module.Index0Float2Write=true
    Module.Index0Float3Write=true
   )
   &ScaleColor(
    Scale Alpha=&FloatFromCurve(
     FloatCurve=DataInterface() {propertyValues {Curve {keys [
         <at 0.1 0.8131 RCIM_Cubic>
         {interpMode=RCIM_Cubic, leaveTangent=-2.4889, tangentMode=RCTM_Break, time=0.3047, value=0.8}
         {arriveTangent=-1.1506, interpMode=RCIM_Cubic, leaveTangent=-1.1506, time=0.532, value=0.181}
         <at 1 0>
        ]}}}
     Scale Curve=&FN_FrontEndFloatModifier(Gameplay=1, Front End=0.6)
    )
    Scale RGB=&VectorFromFloat(Value=&FloatFromCurve(FloatCurve=DataInterface() {propertyValues {Curve {keys [<at 0 10>,<at 0.15 3>,<at 1 1>]}}}))
    ScaleA=true
    ScaleRGB=true
   ) {module=ScaleColor001}
   &ScaleMeshSize(Scale Factor=&VectorFromFloat(Value=&FloatFromCurve(
      FloatCurve=DataInterface() {propertyValues {Curve {keys [<at 0 0.5836>,<at 0.0592 1.6946>,<at 0.3467 2>]}}}
      Scale Curve=&Multiply_Float(
       A=0.115
       B=&Multiply_Float(
        A=0.7084
        B=&FN_LOD_Distance_Remap(Start Distance=350, End Distance=380, Minimum Output=0.7, Maximum Output=1, Clamp Output=true)
       )
      )
     )))
  ]
  rendererClasses [&NiagaraMeshRendererProperties]
  renderers [{rendererClass=&NiagaraMeshRendererProperties, rendererIndex=0}]
 }
}
summary {
 systemName=NS_Emote_SpacePlunge_Hero
 emitters [
  {emitterName=Wave02, rendererClasses [&NiagaraMeshRendererProperties]}
  {bEnabled=false, emitterName=WarpEnergy, rendererClasses [&NiagaraMeshRendererProperties]}
  {emitterName=Hunt_Lightning, rendererClasses [&NiagaraMeshRendererProperties]}
  {bEnabled=false, emitterName=Impact_Flare, rendererClasses [&NiagaraMeshRendererProperties]}
  {bEnabled=false, emitterName=Impact_Flare001, rendererClasses [&NiagaraSpriteRendererProperties]}
 ]
 userVariables [
  User.bIsFrontEnd {type=NiagaraBool}
  User.bIsFrontEndPreview {type=NiagaraBool}
  User.BurstIndex {type=NiagaraInt32}
  User.Variant {type=NiagaraFloat, defaultValue=0, description=""}
 ]
}
